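# Editor utility helper for the Spline Mesh Generator (SMG), exposing editor-only operations as custom events for its parent and spline objects.
# It rebinds selection-change and camera-move delegates to forward callbacks to SMG targets.
# Utilities configure or remove static mesh LODs, dirty data assets by toggling ForDirty, and copy text via Python.
# Texture baking draws each spline's bone- or actor-relative first point, bound-normalized to RGB, into a render target converted to Texture2D.
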EditorUtilityBlueprint /Game/Tools/SplineMeshGenerator/SMG_EditorHelper
graphs [EventGraph {nodes [
   n1 = S_InitialSelectionChangeDetector!() -> (OutputDelegate, then, Target:SMG Spline Object Reference) {is=title K2Node_CustomEvent}
   > n2 = GetSelectionSet(self=Get LevelEditorSubsystem())
   n3 = SC2!() -> (OutputDelegate, then, SelectionSet:Typed Element Selection Set Object Reference) {is=title K2Node_CustomEvent}
   Unbind Event from On Selection Change(execute=@n2.then, self:Typed Element Selection Set Object Reference=@n2, Delegate=@n3) {is=title K2Node_RemoveDelegate}
   > Bind Event to On Selection Change(self:Typed Element Selection Set Object Reference=@n2, Delegate=@n3) {is=title K2Node_AddDelegate}
   n4 = SetupStaticMeshLODs!() -> (OutputDelegate, then, SM:Static Mesh Object Reference) {is=title K2Node_CustomEvent}
   n5 = Get StaticMeshEditorSubsystem()
   n6 = Make Static Mesh Reduction Settings!(PercentTriangles:float=1, ScreenSize:float=0.001) -> (StaticMeshReductionSettings:Static Mesh Reduction Settings Structure) {is=title K2Node_MakeStruct}
   n7 = Make Static Mesh Reduction Settings(PercentTriangles=1, ScreenSize=1)
   > SetLodsWithNotification!(
    execute
    self:Static Mesh Editor Subsystem Object Reference=@n5
    StaticMesh:Static Mesh Object Reference=@n4.SM
    ReductionOptions:"Static Mesh Reduction Options Structure (by ref)"=Make Static Mesh Reduction Options!(
     bAutoComputeLODScreenSize:bool=false
     ReductionSettings:Array of Static Mesh Reduction Settings Structures=Make Array!(
      "[0]":Static Mesh Reduction Settings Structure=@n7
      "[1]":Static Mesh Reduction Settings Structure=@n6
      "[2]":Static Mesh Reduction Settings Structure=@n6
     ) -> (Array:Array of Static Mesh Reduction Settings Structures)
    ) -> (StaticMeshReductionOptions:Static Mesh Reduction Options Structure) {is=title K2Node_MakeStruct}
    bApplyChanges:bool=true
   ) -> (then, ReturnValue:int=0)
   > SetMinLODForQualityLevels!(
    execute
    self:Static Mesh Object Reference=@n4.SM
    QualityLevelMinimumLODs:Map of EPerQualityLevels Enums to Integers=Make Map!(Key 0:EPerQualityLevels Enum=Low, Value 0:int=2, Key 1:EPerQualityLevels Enum=Medium, Value 1:int=1) -> (Map:Map of EPerQualityLevels Enums to Integers) {is=title K2Node_MakeMap}
    Default:int=-1
   ) -> (then)
   n8 = RemoveStaticMeshLODs!() -> (OutputDelegate, then, SM:Static Mesh Object Reference) {is=title K2Node_CustomEvent}
   > RemoveLods!(execute, self:Static Mesh Editor Subsystem Object Reference=@n5, StaticMesh:Static Mesh Object Reference=@n8.SM) -> (then, ReturnValue:bool)
   n9 = P_InitialSelectionChangeDetector!() -> (OutputDelegate, then, Target:SMG Parent Object Reference) {is=title K2Node_CustomEvent}
   > n10 = GetSelectionSet(self=Get LevelEditorSubsystem())
   n11 = SC!() -> (OutputDelegate, then, SelectionSet:Typed Element Selection Set Object Reference) {is=title K2Node_CustomEvent}
   Unbind Event from On Selection Change(execute=@n10.then, self=@n10, Delegate=@n11)
   > Bind Event to On Selection Change(self=@n10, Delegate=@n11)
   n12 = P_SetDataAssetDirty!() -> (OutputDelegate, then, Target:FX Spline Mesh Gen Data Object Reference) {is=title K2Node_CustomEvent}
   > Is Valid(InputObject=@n12.Target)
   > SetEditorProperty(
    Object=@n12.Target
    PropertyName=ForDirty
    PropertyValue=NOT Boolean(A=Get ForDirty!(self:FX Spline Mesh Gen Data Object Reference=@n12.Target) -> (ForDirty:bool))
    ChangeNotifyMode=Always
   )
   n13 = CopyToClipboard!() -> (OutputDelegate, then, text:String) {is=title K2Node_CustomEvent}
   > Execute Python Script!(
    execute
    PythonScript="import os\r\ncommand = 'cmd /c \"echo '+text+'\"|clip'\r\nos.system(command)"
    text:String=Replace(SourceString=@n13.text, From="\r\n", To="&&echo ", SearchCase=CaseSensitive)
   ) -> (then, ReturnValue) {inputs [text]}
   n14 = S_TurnOffCameraOffsetEvent() {is=title K2Node_CustomEvent}
   n15 = CameraMoved!() -> (
    OutputDelegate
    then
    Location:"Vector (by ref)"
    Rotation:"Rotator (by ref)"
    ViewportType:ELevelViewportType Enum=LVT_OrthoXY
    ViewIndex:int
   ) {is=title K2Node_CustomEvent}
   n16 = Get LevelEditorSubsystem()
   Unbind Event from On Editor Camera Moved(execute=@n14.then, self:Level Editor Subsystem Object Reference=@n16, Delegate=@n15) {is=title K2Node_RemoveDelegate}
   n17 = S_TurnOnCameraOffsetEvent!() -> (OutputDelegate, then, Target:SMG Spline Object Reference) {is=title K2Node_CustomEvent}
   > Unbind Event from On Editor Camera Moved(self=@n16, Delegate=@n15)
   > Bind Event to On Editor Camera Moved(self:Level Editor Subsystem Object Reference=@n16, Delegate=@n15) {is=title K2Node_AddDelegate}
   n18 = BakeFirstPointLocalPosTexture!() -> (OutputDelegate, then, TargetTexture:Texture 2D Object Reference, TargetSMGParent:SMG Parent Object Reference, Bound:Vector=~) {is=title K2Node_CustomEvent}
   > Is Valid(InputObject=@n18.TargetTexture)
   > Is Valid(InputObject=@n18.TargetSMGParent)
   n19 = Length(TargetArray:Array of SMG Spline Object References=Get SplineList!(self:SMG Parent Object Reference=@n18.TargetSMGParent) -> (SplineList:Array of SMG Spline Object References))
   n20 = Ceil(A=float / float(A="To Float (Integer)"(InInt=@n19), B=4)) {is=title K2Node_CallFunction}
   > n21 = CreateRenderTarget2D!(
    execute
    WorldContextObject:Object Reference
    Width:int=@n20
    Height:int=4
    Format:ETextureRenderTargetFormat Enum=RTF_RGBA8
    ClearColor:Linear Color Structure=(R=0,G=0,B=0,A=1)
    bAutoGenerateMipMaps:bool=false
    bSupportUAVs:bool=false
   ) -> (then, ReturnValue:Texture Render Target 2D Object Reference)
   > n22 = BeginDrawCanvasToRenderTarget!(execute, WorldContextObject:Object Reference, TextureRenderTarget:Texture Render Target 2D Object Reference=@n21) -> (then, Canvas:Canvas Object Reference, Size:Vector 2D Structure, Context:Draw To Render Target Context Structure)
   > n23 = For Each Loop(Array:Array of SMG Spline Object References=Get SplineList(self=@n18.TargetSMGParent)) -> (Array Element:SMG Spline Object Reference)
   EndDrawCanvasToRenderTarget!(execute=@n23.Completed, WorldContextObject:Object Reference, Context:"Draw To Render Target Context Structure (by ref)"=@n22.Context) -> (then)
   > ConvertRenderTargetToTexture2DEditorOnly(RenderTarget=@n21, Texture=@n18.TargetTexture)
   > SetEditorProperty(
    Object=@n18.TargetTexture
    PropertyName=CompressionSettings
    PropertyValue:TextureCompressionSettings Enum=Get FPLPB_TC() -> (FPLPB_TC:TextureCompressionSettings Enum)
    ChangeNotifyMode=Always
   )
   > SetEditorProperty(
    Object=@n18.TargetTexture
    PropertyName=MipGenSettings
    PropertyValue:TextureMipGenSettings Enum=Get FPLPB_MGS() -> (FPLPB_MGS:TextureMipGenSettings Enum=TMGS_FromTextureGroup)
    ChangeNotifyMode=Always
   )
   > SetEditorProperty(
    Object=@n18.TargetTexture
    PropertyName=TextureGroup
    PropertyValue:TextureGroup Enum=Get FPLPB_TG() -> (FPLPB_TG:TextureGroup Enum)
    ChangeNotifyMode=Always
   )
   > SetEditorProperty(Object=@n18.TargetTexture, PropertyName=NeverStream, PropertyValue:bool=OR Boolean(A=true, B=true), ChangeNotifyMode=Always)
   > SetEditorProperty(Object=@n18.TargetTexture, PropertyName=SRGB, PropertyValue=OR Boolean(A=false, B=false), ChangeNotifyMode=Always)
   > SetEditorProperty(
    Object=@n18.TargetTexture
    PropertyName=X-axisTilingMethod
    PropertyValue:TextureAddress Enum=Get FPLPB_XT() -> (FPLPB_XT:TextureAddress Enum=TA_Wrap)
    ChangeNotifyMode=Always
   )
   > SetEditorProperty(
    Object=@n18.TargetTexture
    PropertyName=Y-axisTilingMethod
    PropertyValue=Get FPLPB_YT() -> (FPLPB_YT:TextureAddress Enum=TA_Wrap)
    ChangeNotifyMode=Always
   )
   > SetEditorProperty(
    Object=@n18.TargetTexture
    PropertyName=Filter
    PropertyValue:TextureFilter Enum=Get FPLPB_F() -> (FPLPB_F:TextureFilter Enum=TF_Nearest)
    ChangeNotifyMode=Always
   )
   n24 = MakeVector2D(
    X="To Float (Integer)"(InInt="% (Integer)"(A=@n23.Array Index, B=@n20) -> (ReturnValue=0))
    Y="To Float (Integer)"(InInt=int / int(A=@n23.Array Index, B=@n20))
   )
   n25 = Get tmpTargetSkMesh!(self:SMG Parent Object Reference=@n18.TargetSMGParent) -> (tmpTargetSkMesh:Skeletal Mesh Component Object Reference)
   n26 = IsValid(Object=@n25)
   n27 = GetSocketTransform(
    self=@n25
    InSocketName="Get TargetBone(Optional)"!(self:SMG Parent Object Reference=@n18.TargetSMGParent) -> ("TargetBone(Optional)":Name=None)
    TransformSpace=RTS_World
   )
   n28 = Get Actor Transform(self=@n18.TargetSMGParent)
   n29 = Get Spline!(self:SMG Spline Object Reference=@n23) -> (Spline:Spline Component Object Reference)
   n30 = vector / vector(
    A:Vector=InverseTransformLocation!(
     T:"Transform (by ref)"=SelectTransform!(A:"Transform (by ref)"=@n27, B:"Transform (by ref)"=@n28, bPickA:bool=@n26) -> (ReturnValue:Transform)
     Location:Vector=GetLocationAtSplinePoint!(self:Spline Component Object Reference=@n29, PointIndex:int=0, CoordinateSpace:ESplineCoordinateSpace Enum=World) -> (ReturnValue:Vector)
    ) -> (ReturnValue:Vector)
    B:Vector=@n18.Bound
   ) -> (ReturnValue:Vector)
   n31 = BreakVector(InVec=vector * vector(A=vector + vector(A=@n30, B="1,1,1"), B:Vector="0.5,0.5,0.5") -> (ReturnValue=~))
   Draw Line!(
    execute=@n23.LoopBody
    self:Canvas Object Reference=@n22
    ScreenPositionA:Vector 2D Structure=@n24
    ScreenPositionB:Vector 2D Structure=vector2d + vector2d(A:Vector 2D Structure=@n24, B:Vector 2D Structure=(X=0.5,Y=0.5)) -> (ReturnValue:Vector 2D Structure) {is=title K2Node_PromotableOperator}
    Thickness:float=0.1
    RenderColor:Linear Color Structure=Make Linear Color!(R:float=@n31, G:float=@n31.Y, B:float=@n31.Z, A:float=0) -> (LinearColor:Linear Color Structure) {is=title K2Node_MakeStruct}
   ) -> (then)
   SelectionSetChangeEvent(execute=@n11.then, self:SMG Parent Object Reference=@n9.Target)
   Check Parent Valid(execute=@n3.then, self:SMG Spline Object Reference=@n1.Target)
   UpdateCameraOffset!(execute=@n15.then, self:SMG Spline Object Reference=@n17.Target, CameraLocation:Vector=@n15.Location) -> (then)
   Event Begin Play()
   Event Tick!() -> (OutputDelegate, then, DeltaSeconds:float)
  ]}]
variables [
 FPLPB_F {type=TextureFilter, default=TF_Nearest}
 FPLPB_MGS {type=TextureMipGenSettings, default=TMGS_NoMipmaps}
 FPLPB_TC {type=TextureCompressionSettings, default=TC_VectorDisplacementmap}
 FPLPB_TG {type=TextureGroup, default=TEXTUREGROUP_EffectsNotFiltered}
 FPLPB_XT {type=TextureAddress, default=TA_Clamp}
 FPLPB_YT {type=TextureAddress, default=TA_Clamp}
]
