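; Editor tool actor that manages SMG_Spline actors, merging their meshes for baking into static meshes used in Niagara FX.
; On rebuild it moves splines with itself and refreshes the skeletal mesh reference.
; CollectMeshes remaps materials, optionally writes seeded random vertex colors and LODs, and packs each spline into a UV1 grid cell.
; FX_SplineMeshGenData assets hold saves and bake targets; clipboard helpers generate Niagara renderer and skeletal location modules.
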(event Custom|SelectionSetChangeEvent
  ;[1 1 1] Comment
  (CallFunction|RefreshSplineListStat))

(event Events|EventOnRebuildGeneratedMesh
  (params (TargetMesh :type Dynamic%20Mesh%20Object%20Reference))
  (bind _self self)
  (bind _returnvalue (Transformation|GetActorTransform _self))
  (bind _savefile (Variables|1.Setup|GetSaveFile))
  ;[1 1 1] Comment
  (Utilities|FlowControl|Sequence
    (:then_0
      (Utilities|FlowControl|Sequence
        (:then_0
          (if (not (== _savefile (Variables|Hide|GetLastFrameSaveFile)))
            (Utilities|Array|Clear (Variables|3.Splines|GetSplineList))
            (Variables|2.SkeletalMeshSetting%28Optional%29|SetTargetSkeletalMesh%28Optional%29 0)
            (Variables|2.SkeletalMeshSetting%28Optional%29|SetTargetBone%28Optional%29 "None")
            (Variables|4.NiagaraHelper|SetCameraOffsetPreviewValue 0.0)
            (Variables|5.Advance|SetbWriteSeededRandomValueIntoVertColorR false)
            (Variables|5.Advance|SetbWriteSeededRandomValueIntoVertColorA false)
            (Variables|5.Advance|SetRandomSeed 0)
            (Variables|5.Advance|SetbLODGeneration false)
            (Variables|5.Advance|SetLOD1SegReduceFactor(Med) 0.7)
            (Variables|5.Advance|SetLOD1DivideReduceFactor(Med) 1.0)
            (Variables|5.Advance|SetLOD2SegReduceFactor(Low) 0.5)
            (Variables|5.Advance|SetLOD2DivideReduceFactor(Low) 1.0)))
        (:then_1
          (Variables|Hide|SetLastFrameSaveFile _savefile))))
    (:then_1
      (Utilities|FlowControl|Sequence
        (:then_0
          (if (not (== (Variables|Hide|GetLastFrameTransform) _returnvalue))
            (for _array_element (Variables|3.Splines|GetSplineList)
              (Transformation|SetActorTransform _array_element (Math|Transform|ComposeTransforms (Math|Transform|ComposeTransforms (Transformation|GetActorTransform _array_element) (Math|Transform|InvertTransform (Variables|Hide|GetLastFrameTransform))) _returnvalue) false false))))
        (:then_1
          (Variables|Hide|SetLastFrameTransform _returnvalue))))
    (:then_2
      (Variables|Hide|SetbHasSaveFile? (Utilities|IsValid (Variables|1.Setup|GetSaveFile) :!class /Script/Engine.KismetSystemLibrary)))
    (:then_3
      (CallFunction|RefreshSplineListStat))
    (:then_4
      (CallFunction|RefreshSkMesh))))

(event Custom|SplineUpdate
  ;[1 1 1] Comment
  (CallFunction|RefreshSkMesh))

(event EventBakeToStaticMesh
  (bind _savefile (Variables|1.Setup|GetSaveFile))
  (bind _targetstaticmesh (Variables|/Game/Tools/SplineMeshGenerator/FX_SplineMeshGenData.FX_SplineMeshGenData_C|GetTargetStaticMesh _savefile))
  (bind _smg_editorhelper (Variables|Default|GetSMG_EditorHelper))
  ;[1 1 1] BakeToStaticMesh
  (Utilities|IsValid _savefile
    (:Is%20Valid
      (Utilities|IsValid _targetstaticmesh
        (:Is%20Valid
          (for _array_element (Variables|3.Splines|GetSplineList)
            (Class|SMGSpline|CorrectTangentandNormal _array_element))
          (bind (_lod1mesh _lod2mesh) (CallFunction|CollectMeshes (Variables|5.Advance|GetbLODGeneration)))
          (bind _returnvalue (DynamicMeshComponent|GetDynamicMesh (Variables|Default|GetTargetDynamicMesh)))
          (GeometryScript|StaticMesh|CopyMeshtoStaticMesh _returnvalue _targetstaticmesh (Utilities|Struct|MakeGeometryScriptCopyMeshtoAssetOptions false false false "DoNothing" false false true true "ReplaceAssetMaterials" (Variables|Hide|GetMaterialList)) (Utilities|Struct|MakeGeometryScriptMeshWriteLOD false)
            (:Success
              (if (Variables|5.Advance|GetbLODGeneration)
                (Class|SMGEditorHelper|SetupStaticMeshLODs _smg_editorhelper _targetstaticmesh)
                (GeometryScript|StaticMesh|CopyMeshtoStaticMesh :FromDynamicMesh _lod1mesh :ToStaticMeshAsset _targetstaticmesh :TargetLOD (Utilities|Struct|MakeGeometryScriptMeshWriteLOD false 1) :bUseSectionMaterials true
                  (:Success
                    (GeometryScript|StaticMesh|CopyMeshtoStaticMesh :FromDynamicMesh _lod2mesh :ToStaticMeshAsset _targetstaticmesh :TargetLOD (Utilities|Struct|MakeGeometryScriptMeshWriteLOD false 2) :bUseSectionMaterials true)))
                (else
                  (Class|SMGEditorHelper|RemoveStaticMeshLODs _smg_editorhelper _targetstaticmesh))))))))))

(fn Bak_1
  (CallFunction|SaveSplinestoBackupSlot 0))

(fn Bak_2
  (CallFunction|SaveSplinestoBackupSlot 1))

(fn Bak_3
  (CallFunction|SaveSplinestoBackupSlot 2))

(fn Bak_4
  (CallFunction|SaveSplinestoBackupSlot 3))

(fn CollectMeshes
  (params (bReturnLODMeshes :type Boolean))
  (bind _splinelist (Variables|3.Splines|GetSplineList))
  (bind _returnvalue (Utilities|Array|Length _splinelist))
  (bind _tmpmaterialslist (LocalVariables|GettmpMaterialsList))
  (bind _returnvalue_2 (Utilities|Array|Length _tmpmaterialslist))
  (bind _tmpmaterialslistperspline (LocalVariables|GettmpMaterialsListPerSpline))
  (bind _returnvalue_3 (Utilities|Array|Length _tmpmaterialslistperspline))
  (bind _targetdynamicmesh (Variables|Default|GetTargetDynamicMesh))
  (bind _returnvalue_4 (Transformation|GetWorldTransform _targetdynamicmesh))
  (bind _rotation (.rotation _returnvalue_4))
  (bind _location (.location _returnvalue_4))
  (bind _returnvalue_5 (Math|Transform|MakeTransform _location _rotation))
  (bind _returnvalue_6 (Math|Conversions|ToFloat(Integer) _returnvalue))
  (bind _returnvalue_7 (/ _returnvalue_6 4.0))
  (bind _returnvalue_8 (Math|Float|Ceil _returnvalue_7))
  (bind _returnvalue_9 (Math|Transform|InvertTransform _returnvalue_5))
  (bind _returnvalue_10 (Math|Conversions|ToFloat(Integer) _returnvalue_8))
  (bind _returnvalue_11 (Math|Conversions|ToFloat(Integer) 4))
  (bind _returnvalue_12 (- _returnvalue_3 1))
  (bind _returnvalue_13 (+ _returnvalue_2 _returnvalue_12))
  (bind _tmpmaterialslist_2 (LocalVariables|GettmpMaterialsList))
  (bind _randomseed (Variables|5.Advance|GetRandomSeed))
  (bind _tmpmaterialslist_3 (LocalVariables|GettmpMaterialsList))
  (bind _tmplod1dm (LocalVariables|GettmpLOD1DM))
  (bind _tmplod2dm (LocalVariables|GettmpLOD2DM))
  (bind _tmpdynamicmesh (LocalVariables|GettmpDynamicMesh))
  (bind _tmpvertcolorlist (LocalVariables|GettmpVertColorList))
  (bind _tmpvertcolorlist_2 (LocalVariables|GettmpVertColorList))
  (bind _tmpvertcolorlist_3 (LocalVariables|GettmpVertColorList))
  (Utilities|IsValid (LocalVariables|GettmpDynamicMesh)
    (:Is%20Valid
      _)
    (:Is%20Not%20Valid
      (bind _returnvalue_14 (Game|ConstructObjectfromClass "/Script/GeometryFramework.DynamicMesh"))
      (LocalVariables|SettmpDynamicMesh _returnvalue_14)))
  (bind _returnvalue_15 (DynamicMeshComponent|GetDynamicMesh (Variables|Default|GetTargetDynamicMesh)))
  (bind _returnvalue_16 (DynamicMesh|Reset _returnvalue_15))
  (GeometryScript|Materials|EnableMaterialIDs _returnvalue_16)
  (if (Variables|GetbReturnLODMeshes)
    (Utilities|FlowControl|Sequence
      (:then_0
        (bind _returnvalue_17 (Game|ConstructObjectfromClass "/Script/GeometryFramework.DynamicMesh"))
        (bind _output_get (LocalVariables|SettmpReturnLOD1DM _returnvalue_17))
        (GeometryScript|Materials|EnableMaterialIDs _output_get))
      (:then_1
        (bind _returnvalue_18 (Game|ConstructObjectfromClass "/Script/GeometryFramework.DynamicMesh"))
        (bind _output_get_2 (LocalVariables|SettmpReturnLOD2DM _returnvalue_18))
        (GeometryScript|Materials|EnableMaterialIDs _output_get_2))
      (:then_2
        _)))
  (Utilities|Array|Clear _tmpmaterialslist_2)
  (GeometryScript|Scene|SetComponentMaterialList (Variables|Default|GetTargetDynamicMesh) _tmpmaterialslist_2)
  (DynamicMeshComponent|SetTangentsType (Variables|Default|GetTargetDynamicMesh) "ExternallyProvided")
  (for (_array_element _array_index) (Variables|3.Splines|GetSplineList)
    (Utilities|IsValid _array_element
      (:Is%20Valid
        (if (Variables|GetbReturnLODMeshes)
          (bind _loddynamicmesh (Class|SMGSpline|GetLODMesh _array_element (Variables|5.Advance|GetLOD1SegReduceFactor(Med)) (Variables|5.Advance|GetLOD1DivideReduceFactor(Med))))
          (LocalVariables|SettmpLOD1DM _loddynamicmesh)
          (bind _loddynamicmesh_2 (Class|SMGSpline|GetLODMesh _array_element (Variables|5.Advance|GetLOD2SegReduceFactor(Low)) (Variables|5.Advance|GetLOD2DivideReduceFactor(Low))))
          (LocalVariables|SettmpLOD2DM _loddynamicmesh_2))
        (bind _returnvalue_19 (DynamicMeshComponent|GetDynamicMesh (Variables|/Script/GeometryFramework.DynamicMeshActor|GetDynamicMeshComponent _array_element)))
        (Utilities|Array|Clear (LocalVariables|GettmpMaterialsListPerSpline))
        (DynamicMesh|Reset _tmpdynamicmesh)
        (bind _copytomeshout (GeometryScript|Decomposition|CopyMeshtoMesh _returnvalue_19 _tmpdynamicmesh))
        (Utilities|FlowControl|Sequence
          (:then_0
            ;[1 1 1] Todo bound preiview
            (Utilities|FlowControl|Sequence
              (:then_0
                (bind _colorlist (GeometryScript|VertexColor|GetMeshPerVertexColors _copytomeshout false))
                (bind _colorarray (GeometryScript|ListUtils|ConvertColorListtoArray _colorlist))
                (bind _output_get_3 (LocalVariables|SettmpVertColorList _colorarray))
                (for (_array_element_2 _array_index_2) _output_get_3
                  (Utilities|FlowControl|Sequence
                    (:then_0
                      (if (Variables|5.Advance|GetbWriteSeededRandomValueIntoVertColorR)
                        (bind _structout (Utilities|Struct|SetmembersinLinearColor _array_element_2 (Math|Random|RandomFloatfromStream (Math|Random|MakeRandomStream (+ _randomseed _array_index)))))
                        (Utilities|Array|SetArrayElem _tmpvertcolorlist_2 _array_index_2 _structout)))
                    (:then_1
                      (if (Variables|5.Advance|GetbWriteSeededRandomValueIntoVertColorA)
                        (bind _structout_2 (Utilities|Struct|SetmembersinLinearColor _array_element_2 (Math|Random|RandomFloatfromStream (Math|Random|MakeRandomStream (+ _randomseed _array_index)))))
                        (Utilities|Array|SetArrayElem _tmpvertcolorlist_2 _array_index_2 _structout_2)))))
                (bind _colorlist_2 (GeometryScript|ListUtils|ConvertArraytoColorList _output_get_3))
                (GeometryScript|VertexColor|SetMeshPerVertexColors _copytomeshout _colorlist_2))
              (:then_1
                (if (Variables|GetbReturnLODMeshes)
                  (Utilities|FlowControl|Sequence
                    (:then_0
                      (bind _colorlist_3 (GeometryScript|VertexColor|GetMeshPerVertexColors _tmplod1dm false))
                      (bind _colorarray_2 (GeometryScript|ListUtils|ConvertColorListtoArray _colorlist_3))
                      (bind _output_get_4 (LocalVariables|SettmpVertColorList _colorarray_2))
                      (for (_array_element_3 _array_index_3) _output_get_4
                        (Utilities|FlowControl|Sequence
                          (:then_0
                            (if (Variables|5.Advance|GetbWriteSeededRandomValueIntoVertColorR)
                              (bind _structout_3 (Utilities|Struct|SetmembersinLinearColor _array_element_3 (Math|Random|RandomFloatfromStream (Math|Random|MakeRandomStream (+ _randomseed _array_index)))))
                              (Utilities|Array|SetArrayElem _tmpvertcolorlist_3 _array_index_3 _structout_3)))
                          (:then_1
                            (if (Variables|5.Advance|GetbWriteSeededRandomValueIntoVertColorA)
                              (bind _structout_4 (Utilities|Struct|SetmembersinLinearColor _array_element_3 (Math|Random|RandomFloatfromStream (Math|Random|MakeRandomStream (+ _randomseed _array_index)))))
                              (Utilities|Array|SetArrayElem _tmpvertcolorlist_3 _array_index_3 _structout_4)))))
                      (bind _colorlist_4 (GeometryScript|ListUtils|ConvertArraytoColorList _output_get_4))
                      (GeometryScript|VertexColor|SetMeshPerVertexColors _tmplod1dm _colorlist_4))
                    (:then_1
                      (bind _colorlist_5 (GeometryScript|VertexColor|GetMeshPerVertexColors _tmplod2dm false))
                      (bind _colorarray_3 (GeometryScript|ListUtils|ConvertColorListtoArray _colorlist_5))
                      (bind _output_get_5 (LocalVariables|SettmpVertColorList _colorarray_3))
                      (for (_array_element_4 _array_index_4) _output_get_5
                        (Utilities|FlowControl|Sequence
                          (:then_0
                            (if (Variables|5.Advance|GetbWriteSeededRandomValueIntoVertColorR)
                              (bind _structout_5 (Utilities|Struct|SetmembersinLinearColor _array_element_4 (Math|Random|RandomFloatfromStream (Math|Random|MakeRandomStream (+ _randomseed _array_index)))))
                              (Utilities|Array|SetArrayElem _tmpvertcolorlist _array_index_4 _structout_5)))
                          (:then_1
                            (if (Variables|5.Advance|GetbWriteSeededRandomValueIntoVertColorA)
                              (bind _structout_6 (Utilities|Struct|SetmembersinLinearColor _array_element_4 (Math|Random|RandomFloatfromStream (Math|Random|MakeRandomStream (+ _randomseed _array_index)))))
                              (Utilities|Array|SetArrayElem _tmpvertcolorlist _array_index_4 _structout_6)))))
                      (bind _colorlist_6 (GeometryScript|ListUtils|ConvertArraytoColorList _output_get_5))
                      (GeometryScript|VertexColor|SetMeshPerVertexColors _tmplod2dm _colorlist_6)))))))
          (:then_1
            ;[1 1 1] Comment
            (Utilities|FlowControl|Sequence
              (:then_0
                (bind (_triangleidlist _bhastriangleidgaps _returnvalue_20) (GeometryScript|MeshQueries|GetAllTriangleIDs _copytomeshout))
                (bind _returnvalue_21 (GeometryScript|UVs|SetNumUVChannels _returnvalue_20 2))
                (bind _indexarray (GeometryScript|ListUtils|ConvertIndexListtoArray _triangleidlist))
                (for _array_element_5 _indexarray
                  (GeometryScript|UVs|SetMeshTriangleUVs _returnvalue_21 1 _array_element_5 (Utilities|Struct|MakeGeometryScriptUVTriangle (Math|Vector2D|MakeVector2D (/ (+ (Math|Conversions|ToFloat(Integer) (Math|Integer|%25%28Integer%29 _array_index _returnvalue_8)) 0.5) _returnvalue_10) (/ (+ (Math|Conversions|ToFloat(Integer) (/ _array_index _returnvalue_8)) 0.5) _returnvalue_11)) (Math|Vector2D|MakeVector2D (/ (+ (Math|Conversions|ToFloat(Integer) (Math|Integer|%25%28Integer%29 _array_index _returnvalue_8)) 0.5) _returnvalue_10) (/ (+ (Math|Conversions|ToFloat(Integer) (/ _array_index _returnvalue_8)) 0.5) _returnvalue_11)) (Math|Vector2D|MakeVector2D (/ (+ (Math|Conversions|ToFloat(Integer) (Math|Integer|%25%28Integer%29 _array_index _returnvalue_8)) 0.5) _returnvalue_10) (/ (+ (Math|Conversions|ToFloat(Integer) (/ _array_index _returnvalue_8)) 0.5) _returnvalue_11))))))
              (:then_1
                (if (Variables|GetbReturnLODMeshes)
                  (Utilities|FlowControl|Sequence
                    (:then_0
                      (bind (_triangleidlist_2 _bhastriangleidgaps_2 _returnvalue_22) (GeometryScript|MeshQueries|GetAllTriangleIDs (LocalVariables|GettmpLOD1DM)))
                      (bind _returnvalue_23 (GeometryScript|UVs|SetNumUVChannels _returnvalue_22 2))
                      (bind _indexarray_2 (GeometryScript|ListUtils|ConvertIndexListtoArray _triangleidlist_2))
                      (for _array_element_6 _indexarray_2
                        (GeometryScript|UVs|SetMeshTriangleUVs _returnvalue_23 1 _array_element_6 (Utilities|Struct|MakeGeometryScriptUVTriangle (Math|Vector2D|MakeVector2D (/ (+ (Math|Conversions|ToFloat(Integer) (Math|Integer|%25%28Integer%29 _array_index _returnvalue_8)) 0.5) _returnvalue_10) (/ (+ (Math|Conversions|ToFloat(Integer) (/ _array_index _returnvalue_8)) 0.5) _returnvalue_11)) (Math|Vector2D|MakeVector2D (/ (+ (Math|Conversions|ToFloat(Integer) (Math|Integer|%25%28Integer%29 _array_index _returnvalue_8)) 0.5) _returnvalue_10) (/ (+ (Math|Conversions|ToFloat(Integer) (/ _array_index _returnvalue_8)) 0.5) _returnvalue_11)) (Math|Vector2D|MakeVector2D (/ (+ (Math|Conversions|ToFloat(Integer) (Math|Integer|%25%28Integer%29 _array_index _returnvalue_8)) 0.5) _returnvalue_10) (/ (+ (Math|Conversions|ToFloat(Integer) (/ _array_index _returnvalue_8)) 0.5) _returnvalue_11))))))
                    (:then_1
                      (bind (_triangleidlist_3 _bhastriangleidgaps_3 _returnvalue_24) (GeometryScript|MeshQueries|GetAllTriangleIDs (LocalVariables|GettmpLOD2DM)))
                      (bind _returnvalue_25 (GeometryScript|UVs|SetNumUVChannels _returnvalue_24 2))
                      (bind _indexarray_3 (GeometryScript|ListUtils|ConvertIndexListtoArray _triangleidlist_3))
                      (for _array_element_7 _indexarray_3
                        (GeometryScript|UVs|SetMeshTriangleUVs _returnvalue_25 1 _array_element_7 (Utilities|Struct|MakeGeometryScriptUVTriangle (Math|Vector2D|MakeVector2D (/ (+ (Math|Conversions|ToFloat(Integer) (Math|Integer|%25%28Integer%29 _array_index _returnvalue_8)) 0.5) _returnvalue_10) (/ (+ (Math|Conversions|ToFloat(Integer) (/ _array_index _returnvalue_8)) 0.5) _returnvalue_11)) (Math|Vector2D|MakeVector2D (/ (+ (Math|Conversions|ToFloat(Integer) (Math|Integer|%25%28Integer%29 _array_index _returnvalue_8)) 0.5) _returnvalue_10) (/ (+ (Math|Conversions|ToFloat(Integer) (/ _array_index _returnvalue_8)) 0.5) _returnvalue_11)) (Math|Vector2D|MakeVector2D (/ (+ (Math|Conversions|ToFloat(Integer) (Math|Integer|%25%28Integer%29 _array_index _returnvalue_8)) 0.5) _returnvalue_10) (/ (+ (Math|Conversions|ToFloat(Integer) (/ _array_index _returnvalue_8)) 0.5) _returnvalue_11)))))))))))
          (:then_4
            (for _index (range (+ (.ReturnValue (GeometryScript|MeshQueries|GetMaxMaterialID _returnvalue_19)) 1))
              ;[1 1 1] Comment
              (Utilities|FlowControl|Sequence
                (:then_0
                  (if (not (>= (Utilities|Array|FindItem _tmpmaterialslist (Utilities|Array|Get(acopy) (Variables|/Game/Tools/SplineMeshGenerator/SMG_Spline.SMG_Spline_C|GetMaterials _array_element) _index)) 0))
                    (Utilities|Array|Add (LocalVariables|GettmpMaterialsListPerSpline) (Utilities|Array|Get(acopy) (Variables|/Game/Tools/SplineMeshGenerator/SMG_Spline.SMG_Spline_C|GetMaterials _array_element) _index))))
                (:then_1
                  (GeometryScript|Materials|RemapMaterialIDs _copytomeshout _index (Math|Integer|SelectInt (Utilities|Array|FindItem _tmpmaterialslist (Utilities|Array|Get(acopy) (Variables|/Game/Tools/SplineMeshGenerator/SMG_Spline.SMG_Spline_C|GetMaterials _array_element) _index)) _returnvalue_13 (>= (Utilities|Array|FindItem _tmpmaterialslist (Utilities|Array|Get(acopy) (Variables|/Game/Tools/SplineMeshGenerator/SMG_Spline.SMG_Spline_C|GetMaterials _array_element) _index)) 0)))
                  (if (Variables|GetbReturnLODMeshes)
                    (Utilities|FlowControl|Sequence
                      (:then_0
                        (GeometryScript|Materials|RemapMaterialIDs (LocalVariables|GettmpLOD1DM) _index (Math|Integer|SelectInt (Utilities|Array|FindItem _tmpmaterialslist (Utilities|Array|Get(acopy) (Variables|/Game/Tools/SplineMeshGenerator/SMG_Spline.SMG_Spline_C|GetMaterials _array_element) _index)) _returnvalue_13 (>= (Utilities|Array|FindItem _tmpmaterialslist (Utilities|Array|Get(acopy) (Variables|/Game/Tools/SplineMeshGenerator/SMG_Spline.SMG_Spline_C|GetMaterials _array_element) _index)) 0))))
                      (:then_1
                        (GeometryScript|Materials|RemapMaterialIDs (LocalVariables|GettmpLOD2DM) _index (Math|Integer|SelectInt (Utilities|Array|FindItem _tmpmaterialslist (Utilities|Array|Get(acopy) (Variables|/Game/Tools/SplineMeshGenerator/SMG_Spline.SMG_Spline_C|GetMaterials _array_element) _index)) _returnvalue_13 (>= (Utilities|Array|FindItem _tmpmaterialslist (Utilities|Array|Get(acopy) (Variables|/Game/Tools/SplineMeshGenerator/SMG_Spline.SMG_Spline_C|GetMaterials _array_element) _index)) 0))))))))))
          (:then_5
            (GeometryScript|MeshEdits|AppendMesh _returnvalue_15 _copytomeshout (Math|Transform|MakeTransform (.location (Math|Transform|ComposeTransforms (Transformation|GetActorTransform _array_element) _returnvalue_9)) (.rotation (Math|Transform|ComposeTransforms (Transformation|GetActorTransform _array_element) _returnvalue_9))) false "(CombineMode=EnableAllMatching)")
            (Utilities|Array|AppendArray (LocalVariables|GettmpMaterialsList) (LocalVariables|GettmpMaterialsListPerSpline))
            (if (Variables|GetbReturnLODMeshes)
              (Utilities|FlowControl|Sequence
                (:then_0
                  (GeometryScript|MeshEdits|AppendMesh (LocalVariables|GettmpReturnLOD1DM) (LocalVariables|GettmpLOD1DM) (Math|Transform|MakeTransform (.location (Math|Transform|ComposeTransforms (Transformation|GetActorTransform _array_element) _returnvalue_9)) (.rotation (Math|Transform|ComposeTransforms (Transformation|GetActorTransform _array_element) _returnvalue_9))) false "(CombineMode=EnableAllMatching)")
                  (GeometryScript|MeshEdits|AppendMesh (LocalVariables|GettmpReturnLOD2DM) (LocalVariables|GettmpLOD2DM) (Math|Transform|MakeTransform (.location (Math|Transform|ComposeTransforms (Transformation|GetActorTransform _array_element) _returnvalue_9)) (.rotation (Math|Transform|ComposeTransforms (Transformation|GetActorTransform _array_element) _returnvalue_9))) false "(CombineMode=EnableAllMatching)")))))))))
  (GeometryScript|Scene|SetComponentMaterialList (Variables|Default|GetTargetDynamicMesh) _tmpmaterialslist_3)
  (Variables|Hide|SetMaterialList _tmpmaterialslist_3)
  (if (Variables|GetbReturnLODMeshes)
    (return (LocalVariables|GettmpReturnLOD1DM) (LocalVariables|GettmpReturnLOD2DM))
    (else
      (return 0 0))))

(fn CollectSplineSaveList
  (for _array_element (Variables|3.Splines|GetSplineList)
    (Utilities|Array|Clear (LocalVariables|GettmpPoints))
    (for _index (range (Spline|GetNumberOfSplinePoints (Variables|/Game/Tools/SplineMeshGenerator/SMG_Spline.SMG_Spline_C|GetSpline _array_element)))
      (Utilities|Array|Add (LocalVariables|GettmpPoints) (Utilities|Struct|MakeSMGSplinePoint (Spline|GetLocationatSplinePoint (Variables|/Game/Tools/SplineMeshGenerator/SMG_Spline.SMG_Spline_C|GetSpline _array_element) _index "Local") (Spline|GetRotationatSplinePoint (Variables|/Game/Tools/SplineMeshGenerator/SMG_Spline.SMG_Spline_C|GetSpline _array_element) _index "Local") (Spline|GetScaleatSplinePoint (Variables|/Game/Tools/SplineMeshGenerator/SMG_Spline.SMG_Spline_C|GetSpline _array_element) _index) (Spline|GetArriveTangentatSplinePoint (Variables|/Game/Tools/SplineMeshGenerator/SMG_Spline.SMG_Spline_C|GetSpline _array_element) _index "Local") (Spline|GetLeaveTangentatSplinePoint (Variables|/Game/Tools/SplineMeshGenerator/SMG_Spline.SMG_Spline_C|GetSpline _array_element) _index "Local") (Spline|GetSplinePointType (Variables|/Game/Tools/SplineMeshGenerator/SMG_Spline.SMG_Spline_C|GetSpline _array_element) _index))))
    (Utilities|Array|Add (LocalVariables|GettmpSplines) (Utilities|Struct|MakeSMGSplineSave (LocalVariables|GettmpPoints) (Variables|/Game/Tools/SplineMeshGenerator/SMG_Spline.SMG_Spline_C|GetSegment _array_element) (Variables|/Game/Tools/SplineMeshGenerator/SMG_Spline.SMG_Spline_C|GetDivide _array_element) (Variables|/Game/Tools/SplineMeshGenerator/SMG_Spline.SMG_Spline_C|GetMaterials _array_element) (Math|Transform|ComposeTransforms (Transformation|GetActorTransform _array_element) (Math|Transform|InvertTransform (Transformation|GetActorTransform))) (Variables|/Game/Tools/SplineMeshGenerator/SMG_Spline.SMG_Spline_C|GetVertColor_B _array_element) (Variables|/Game/Tools/SplineMeshGenerator/SMG_Spline.SMG_Spline_C|GetVertColor_R _array_element) (Variables|/Game/Tools/SplineMeshGenerator/SMG_Spline.SMG_Spline_C|GetVertColor_G _array_element) (Variables|/Game/Tools/SplineMeshGenerator/SMG_Spline.SMG_Spline_C|GetVertColor_A _array_element) (Variables|/Game/Tools/SplineMeshGenerator/SMG_Spline.SMG_Spline_C|GetSwizzleUV _array_element) (Variables|/Game/Tools/SplineMeshGenerator/SMG_Spline.SMG_Spline_C|GetVertColorCurve _array_element))))
  (return (LocalVariables|GettmpSplines)))

(fn CopyNiagaraRender
  (if (Variables|Hide|GetbHasSaveFile?)
    (bind _returnvalue (EditorScripting|MessageDialog|ShowMessageDialog (Utilities|Text|MakeLiteralText "NSLOCTEXT(\"[DC128C414BB6600D6E3E1FB639C86A94]\", \"F0E95BCE4974789965849695DF84D61F\", \"Tip\")") (Utilities|Text|MakeLiteralText "NSLOCTEXT(\"[DC128C414BB6600D6E3E1FB639C86A94]\", \"0863364E4EFE7249531F16ABB5BD9174\", \"Do you want to save and bake splines before copy NS module?\")") "YesNo" "Yes"))
    (switch Utilities|FlowControl|Switch|SwitchonEAppReturnType _returnvalue
      (:No
        _)
      (:Yes
        (1.Setup|SaveandBake)))
    (bind _returnvalue_2 (EditorScripting|Asset|GetPathNameforLoadedAsset (Variables|/Game/Tools/SplineMeshGenerator/FX_SplineMeshGenData.FX_SplineMeshGenData_C|GetTargetStaticMesh (Variables|1.Setup|GetSaveFile)) :!class /Script/EditorScriptingUtilities.EditorAssetLibrary))
    (Class|SMGEditorHelper|CopytoClipboard (Variables|Default|GetSMG_EditorHelper) (Utilities|String|Replace "Begin Object Class=/Script/NiagaraEditor.NiagaraClipboardContent Name=\"NiagaraClipboardContent_0\"
   Begin Object Class=/Script/Niagara.NiagaraMeshRendererProperties Name=\"NiagaraMeshRendererProperties_0\"
   End Object
   Begin Object Name=\"NiagaraMeshRendererProperties_0\"
      Meshes(0)=(Mesh=/Script/Engine.StaticMesh'\"Change-This-Part\"')
      PositionBinding=(bBindingExistsOnSource=True)
      MaterialRandomBinding=(bBindingExistsOnSource=True)
      CustomSortingBinding=(bBindingExistsOnSource=True)
      NormalizedAgeBinding=(bBindingExistsOnSource=True)
      RendererEnabledBinding=(ParamMapVariable=(TypeDefHandle=(RegisteredTypeIndex=78)),DataSetVariable=(TypeDefHandle=(RegisteredTypeIndex=78)),RootVariable=(TypeDefHandle=(RegisteredTypeIndex=78)),BoundVariable=(TypeDefHandle=(RegisteredTypeIndex=81)),bIsCachedParticleValue=False)
      OuterEmitterVersion=77B881A6E3912B83D6F96C2A1569A1E6
      MergeId=ED6F21C84BF52252E16FC4BCBD81E676
   End Object
   Renderers(0)=/Script/Niagara.NiagaraMeshRendererProperties'\"NiagaraMeshRendererProperties_0\"'
End Object" "Change-This-Part" _returnvalue_2 "CaseSensitive"))
    (else
      (EditorScripting|MessageDialog|ShowMessageDialog (Utilities|Text|MakeLiteralText "NSLOCTEXT(\"[DC128C414BB6600D6E3E1FB639C86A94]\", \"79B570B5459E38433965759D45C26665\", \"Warning\")") (Utilities|Text|MakeLiteralText "NSLOCTEXT(\"[DC128C414BB6600D6E3E1FB639C86A94]\", \"75088FF848303802D75B6EAD82F15559\", \"This operation requires a valid save file\")") "Ok" "Ok"))))

(fn CopyNiagaraSKLocation(IfHasTargetSKMesh)
  (bind _savefile (Variables|1.Setup|GetSaveFile))
  (if (Variables|Hide|GetbHasSaveFile?)
    (Utilities|IsValid (Variables|/Game/Tools/SplineMeshGenerator/FX_SplineMeshGenData.FX_SplineMeshGenData_C|GetTargetSkeletalMesh (Variables|1.Setup|GetSaveFile))
      (:Is%20Valid
        (bind _returnvalue (EditorScripting|MessageDialog|ShowMessageDialog (Utilities|Text|MakeLiteralText "NSLOCTEXT(\"[DC128C414BB6600D6E3E1FB639C86A94]\", \"F0E95BCE4974789965849695DF84D61F\", \"Tip\")") (Utilities|Text|MakeLiteralText "NSLOCTEXT(\"[DC128C414BB6600D6E3E1FB639C86A94]\", \"0863364E4EFE7249531F16ABB5BD9174\", \"Do you want to save and bake splines before copy NS module?\")") "YesNo" "Yes"))
        (switch Utilities|FlowControl|Switch|SwitchonEAppReturnType _returnvalue
          (:No
            _)
          (:Yes
            (1.Setup|SaveandBake)))
        (bind _returnvalue_2 (EditorScripting|Asset|GetPathNameforLoadedAsset (Variables|/Game/Tools/SplineMeshGenerator/FX_SplineMeshGenData.FX_SplineMeshGenData_C|GetTargetSkeletalMesh _savefile) :!class /Script/EditorScriptingUtilities.EditorAssetLibrary))
        (Class|SMGEditorHelper|CopytoClipboard (Variables|Default|GetSMG_EditorHelper) (Utilities|String|Replace (Utilities|String|Replace "Begin Object Class=/Script/NiagaraEditor.NiagaraClipboardContent Name=\"NiagaraClipboardContent_6\"
   Begin Object Class=/Script/NiagaraEditor.NiagaraClipboardFunction Name=\"NiagaraClipboardFunction_0\"
      Begin Object Class=/Script/NiagaraEditor.NiagaraClipboardFunctionInput Name=\"NiagaraClipboardFunctionInput_17\"
      End Object
      Begin Object Class=/Script/NiagaraEditor.NiagaraClipboardFunctionInput Name=\"NiagaraClipboardFunctionInput_16\"
      End Object
      Begin Object Class=/Script/NiagaraEditor.NiagaraClipboardFunctionInput Name=\"NiagaraClipboardFunctionInput_15\"
      End Object
      Begin Object Class=/Script/NiagaraEditor.NiagaraClipboardFunctionInput Name=\"NiagaraClipboardFunctionInput_14\"
      End Object
      Begin Object Class=/Script/NiagaraEditor.NiagaraClipboardFunctionInput Name=\"NiagaraClipboardFunctionInput_13\"
      End Object
      Begin Object Class=/Script/NiagaraEditor.NiagaraClipboardFunctionInput Name=\"NiagaraClipboardFunctionInput_12\"
      End Object
      Begin Object Class=/Script/NiagaraEditor.NiagaraClipboardFunctionInput Name=\"NiagaraClipboardFunctionInput_11\"
      End Object
      Begin Object Class=/Script/NiagaraEditor.NiagaraClipboardFunctionInput Name=\"NiagaraClipboardFunctionInput_10\"
      End Object
      Begin Object Class=/Script/NiagaraEditor.NiagaraClipboardFunctionInput Name=\"NiagaraClipboardFunctionInput_9\"
      End Object
      Begin Object Class=/Script/NiagaraEditor.NiagaraClipboardFunctionInput Name=\"NiagaraClipboardFunctionInput_8\"
      End Object
      Begin Object Class=/Script/NiagaraEditor.NiagaraClipboardFunctionInput Name=\"NiagaraClipboardFunctionInput_7\"
      End Object
      Begin Object Class=/Script/NiagaraEditor.NiagaraClipboardFunctionInput Name=\"NiagaraClipboardFunctionInput_6\"
      End Object
      Begin Object Class=/Script/NiagaraEditor.NiagaraClipboardFunctionInput Name=\"NiagaraClipboardFunctionInput_5\"
      End Object
      Begin Object Class=/Script/NiagaraEditor.NiagaraClipboardFunctionInput Name=\"NiagaraClipboardFunctionInput_4\"
      End Object
      Begin Object Class=/Script/NiagaraEditor.NiagaraClipboardFunctionInput Name=\"NiagaraClipboardFunctionInput_3\"
      End Object
      Begin Object Class=/Script/NiagaraEditor.NiagaraClipboardFunctionInput Name=\"NiagaraClipboardFunctionInput_2\"
      End Object
      Begin Object Class=/Script/NiagaraEditor.NiagaraClipboardFunctionInput Name=\"NiagaraClipboardFunctionInput_1\"
         Begin Object Class=/Script/Niagara.NiagaraDataInterfaceSkeletalMesh Name=\"NiagaraDataInterfaceSkeletalMesh_0\"
         End Object
      End Object
      Begin Object Class=/Script/NiagaraEditor.NiagaraClipboardFunctionInput Name=\"NiagaraClipboardFunctionInput_0\"
      End Object
   End Object
   Begin Object Name=\"NiagaraClipboardFunction_0\"
      Begin Object Name=\"NiagaraClipboardFunctionInput_4\"
         InputName=\"Bone / Socket Index\"
         InputType=(ClassStructOrEnum=/Script/CoreUObject.ScriptStruct'\"/Script/Niagara.NiagaraInt32\"',UnderlyingType=2)
         Local(0)=0
         Local(1)=0
         Local(2)=0
         Local(3)=0
      End Object
      Begin Object Name=\"NiagaraClipboardFunctionInput_3\"
         InputName=\"Bone Sampling Mode\"
         InputType=(ClassStructOrEnum=/Script/Engine.UserDefinedEnum'\"/Niagara/Enums/ENiagaraBoneSamplingMode.ENiagaraBoneSamplingMode\"',UnderlyingType=3)
         Local(0)=3
         Local(1)=0
         Local(2)=0
         Local(3)=0
      End Object
      Begin Object Name=\"NiagaraClipboardFunctionInput_2\"
         InputName=\"Mesh Sampling Type\"
         InputType=(ClassStructOrEnum=/Script/Engine.UserDefinedEnum'\"/Niagara/Enums/ENiagaraSkelSamplingModeFull.ENiagaraSkelSamplingModeFull\"',UnderlyingType=3)
         Local(0)=0
         Local(1)=0
         Local(2)=0
         Local(3)=0
      End Object
      Begin Object Name=\"NiagaraClipboardFunctionInput_1\"
         Begin Object Name=\"NiagaraDataInterfaceSkeletalMesh_0\"
            PreviewMesh=\"Change-This-SKMesh\"
            FilteredBones(0)=\"Change-This-BoneName\"
            MergeId=AC364AAD4D3468BBB1B76D9ABCDB497A
         End Object
         InputName=\"Skeletal Mesh\"
         InputType=(ClassStructOrEnum=/Script/CoreUObject.Class'\"/Script/Niagara.NiagaraDataInterfaceSkeletalMesh\"',UnderlyingType=1)
         ValueMode=Data
         Data=/Script/Niagara.NiagaraDataInterfaceSkeletalMesh'\"NiagaraDataInterfaceSkeletalMesh_0\"'
      End Object
      FunctionName=\"SkeletalMeshLocation\"
      Script=\"/Niagara/Modules/Spawn/Location/V2/SkeletalMeshLocation.SkeletalMeshLocation\"
      Inputs(0)=/Script/NiagaraEditor.NiagaraClipboardFunctionInput'\"NiagaraClipboardFunctionInput_0\"'
      Inputs(1)=/Script/NiagaraEditor.NiagaraClipboardFunctionInput'\"NiagaraClipboardFunctionInput_1\"'
      Inputs(2)=/Script/NiagaraEditor.NiagaraClipboardFunctionInput'\"NiagaraClipboardFunctionInput_2\"'
      Inputs(3)=/Script/NiagaraEditor.NiagaraClipboardFunctionInput'\"NiagaraClipboardFunctionInput_3\"'
      Inputs(4)=/Script/NiagaraEditor.NiagaraClipboardFunctionInput'\"NiagaraClipboardFunctionInput_4\"'
      Inputs(5)=/Script/NiagaraEditor.NiagaraClipboardFunctionInput'\"NiagaraClipboardFunctionInput_5\"'
      Inputs(6)=/Script/NiagaraEditor.NiagaraClipboardFunctionInput'\"NiagaraClipboardFunctionInput_6\"'
      Inputs(7)=/Script/NiagaraEditor.NiagaraClipboardFunctionInput'\"NiagaraClipboardFunctionInput_7\"'
      Inputs(8)=/Script/NiagaraEditor.NiagaraClipboardFunctionInput'\"NiagaraClipboardFunctionInput_8\"'
      Inputs(9)=/Script/NiagaraEditor.NiagaraClipboardFunctionInput'\"NiagaraClipboardFunctionInput_9\"'
      Inputs(10)=/Script/NiagaraEditor.NiagaraClipboardFunctionInput'\"NiagaraClipboardFunctionInput_10\"'
      Inputs(11)=/Script/NiagaraEditor.NiagaraClipboardFunctionInput'\"NiagaraClipboardFunctionInput_11\"'
      Inputs(12)=/Script/NiagaraEditor.NiagaraClipboardFunctionInput'\"NiagaraClipboardFunctionInput_12\"'
      Inputs(13)=/Script/NiagaraEditor.NiagaraClipboardFunctionInput'\"NiagaraClipboardFunctionInput_13\"'
      Inputs(14)=/Script/NiagaraEditor.NiagaraClipboardFunctionInput'\"NiagaraClipboardFunctionInput_14\"'
      Inputs(15)=/Script/NiagaraEditor.NiagaraClipboardFunctionInput'\"NiagaraClipboardFunctionInput_15\"'
      Inputs(16)=/Script/NiagaraEditor.NiagaraClipboardFunctionInput'\"NiagaraClipboardFunctionInput_16\"'
      Inputs(17)=/Script/NiagaraEditor.NiagaraClipboardFunctionInput'\"NiagaraClipboardFunctionInput_17\"'
      ScriptVersion=B94C0E6D48081AF8DCE0CD919A3582C4
   End Object
   Functions(0)=/Script/NiagaraEditor.NiagaraClipboardFunction'\"NiagaraClipboardFunction_0\"'
End Object
" "Change-This-SKMesh" _returnvalue_2 "CaseSensitive") "Change-This-BoneName" (Utilities|String|ToString(Name) (Variables|/Game/Tools/SplineMeshGenerator/FX_SplineMeshGenData.FX_SplineMeshGenData_C|GetTargetSocket _savefile)) "CaseSensitive")))
      (:Is%20Not%20Valid
        (EditorScripting|MessageDialog|ShowMessageDialog (Utilities|Text|MakeLiteralText "NSLOCTEXT(\"[DC128C414BB6600D6E3E1FB639C86A94]\", \"79B570B5459E38433965759D45C26665\", \"Warning\")") (Utilities|Text|MakeLiteralText "NSLOCTEXT(\"[DC128C414BB6600D6E3E1FB639C86A94]\", \"3A1798454B1D09AC41EA3FBE64EDD97C\", \"This operation requires a valid SK mesh\")") "Ok" "Ok")))
    (else
      (EditorScripting|MessageDialog|ShowMessageDialog (Utilities|Text|MakeLiteralText "NSLOCTEXT(\"[DC128C414BB6600D6E3E1FB639C86A94]\", \"79B570B5459E38433965759D45C26665\", \"Warning\")") (Utilities|Text|MakeLiteralText "NSLOCTEXT(\"[DC128C414BB6600D6E3E1FB639C86A94]\", \"75088FF848303802D75B6EAD82F15559\", \"This operation requires a valid save file\")") "Ok" "Ok"))))

(fn CreateSaveFile
  (bind _savefile (Variables|1.Setup|GetSaveFile))
  (bind _targetstaticmesh (Variables|/Game/Tools/SplineMeshGenerator/FX_SplineMeshGenData.FX_SplineMeshGenData_C|GetTargetStaticMesh _savefile))
  (bind _returnvalue (EditorScripting|MessageDialog|ShowMessageDialog (Utilities|Text|MakeLiteralText "NSLOCTEXT(\"[DC128C414BB6600D6E3E1FB639C86A94]\", \"79B570B5459E38433965759D45C26665\", \"Warning\")") (Utilities|Text|MakeLiteralText "NSLOCTEXT(\"[DC128C414BB6600D6E3E1FB639C86A94]\", \"238A580D4E97A2215F20E09B95EB9A76\", \"This operation cannot be undo.The unsaved progress will be lost permanently.Confirm creating?\")") "YesNo" "No"))
  (switch Utilities|FlowControl|Switch|SwitchonEAppReturnType _returnvalue
    (:Yes
      (Utilities|FlowControl|Sequence
        (:then_0
          (if (Utilities|String|IsEmpty (Variables|1.Setup|GetCreateSaveFileName))
            (EditorScripting|MessageDialog|ShowMessageDialog (Utilities|Text|MakeLiteralText "NSLOCTEXT(\"[DC128C414BB6600D6E3E1FB639C86A94]\", \"79B570B5459E38433965759D45C26665\", \"Warning\")") (Utilities|Text|MakeLiteralText "NSLOCTEXT(\"[DC128C414BB6600D6E3E1FB639C86A94]\", \"815381A74E04FD969B989D92E1604245\", \"Please input save file name.\")") "Ok" "Ok")
            (return)))
        (:then_1
          (bind _returnvalue_2 (EditorScripting|Asset|DoesAssetExist (Utilities|String|Append "/Game/NeverCook/FX/SMG/Saves/" (Variables|1.Setup|GetCreateSaveFileName)) :!class /Script/EditorScriptingUtilities.EditorAssetLibrary))
          (bind _returnvalue_3 (EditorScripting|Asset|DoesAssetExist (Utilities|String|Append "/Game/Tools/SplineMeshGenerator/Saves/Meshes/" (Utilities|String|Append "SM_" (Variables|1.Setup|GetCreateSaveFileName))) :!class /Script/EditorScriptingUtilities.EditorAssetLibrary))
          (if (or _returnvalue_3 _returnvalue_2)
            (EditorScripting|MessageDialog|ShowMessageDialog (Utilities|Text|MakeLiteralText "NSLOCTEXT(\"[DC128C414BB6600D6E3E1FB639C86A94]\", \"C59F9C7E4ACF109743B8DDA48FBE2EB3\", \"Warning\")") (Utilities|Text|MakeLiteralText "NSLOCTEXT(\"[DC128C414BB6600D6E3E1FB639C86A94]\", \"DB63211A4A9FB2A513D9CA86B89D03AA\", \"Target save file already exist.Please change name.\")") "Ok" "Ok")
            (return)))
        (:then_2
          (bind _returnvalue_4 (EditorScripting|Asset|DuplicateAsset "/Game/NeverCook/FX/SMG/DONOTUSE_TemplateDataAsset" (Utilities|String|Append "/Game/NeverCook/FX/SMG/Saves/" (Variables|1.Setup|GetCreateSaveFileName)) :!class /Script/EditorScriptingUtilities.EditorAssetLibrary))
          (Utilities|IsValid _returnvalue_4
            (:Is%20Valid
              (bind _returnvalue_5 (EditorScripting|Asset|DuplicateAsset "/Game/NeverCook/FX/SMG/SMG_TemplateStaticMesh" (Utilities|String|Append "/Game/Tools/SplineMeshGenerator/Saves/Meshes/" (Utilities|String|Append "SM_" (Variables|1.Setup|GetCreateSaveFileName))) :!class /Script/EditorScriptingUtilities.EditorAssetLibrary))
              (Utilities|IsValid _returnvalue_5
                (:Is%20Valid
                  (Variables|/Game/Tools/SplineMeshGenerator/FX_SplineMeshGenData.FX_SplineMeshGenData_C|SetTargetStaticMesh (Utilities|Casting|CastToStaticMesh _returnvalue_5) (Utilities|Casting|CastToFX_SplineMeshGenData _returnvalue_4))
                  (Variables|1.Setup|SetSaveFile (Utilities|Casting|CastToFX_SplineMeshGenData _returnvalue_4))
                  (Variables|Hide|SetbHasSaveFile? true)
                  (Variables|Hide|SetLastFrameSaveFile (Utilities|Casting|CastToFX_SplineMeshGenData _returnvalue_4))
                  (EditorScripting|Asset|SaveLoadedAsset _savefile false :!class /Script/EditorScriptingUtilities.EditorAssetLibrary)
                  (Utilities|IsValid _targetstaticmesh
                    (:Is%20Valid
                      (EditorScripting|Asset|SaveLoadedAsset _targetstaticmesh false :!class /Script/EditorScriptingUtilities.EditorAssetLibrary))))
                (:Is%20Not%20Valid
                  _)))
            (:Is%20Not%20Valid
              _))
          (EditorScripting|MessageDialog|ShowMessageDialog (Utilities|Text|MakeLiteralText "NSLOCTEXT(\"[DC128C414BB6600D6E3E1FB639C86A94]\", \"33F316FB41EDDA440828FFBED451296D\", \"Error\")") (Utilities|Text|MakeLiteralText "NSLOCTEXT(\"[DC128C414BB6600D6E3E1FB639C86A94]\", \"328F852146FECA9505C0348488453E6D\", \"Create file failed, please DM Yi.Xie on slack.\")") "Ok" "Ok"))))))

(fn CreateNewSpline
  (if (Variables|Hide|GetbHasSaveFile?)
    (bind _returnvalue (Game|SpawnActorfromClass "/Game/Tools/SplineMeshGenerator/SMG_Spline.SMG_Spline_C" (Transformation|GetActorTransform) "AlwaysSpawn" "MultiplyWithRoot" 0 false 0))
    (Transformation|AttachActorToActor _returnvalue self "None" "KeepWorld" "KeepWorld" "KeepWorld" true)
    (Utilities|Array|Add (Variables|3.Splines|GetSplineList) _returnvalue)
    (CallFunction|RefreshSplineListStat)
    (else
      (EditorScripting|MessageDialog|ShowMessageDialog (Utilities|Text|MakeLiteralText "NSLOCTEXT(\"[DC128C414BB6600D6E3E1FB639C86A94]\", \"79B570B5459E38433965759D45C26665\", \"Warning\")") (Utilities|Text|MakeLiteralText "NSLOCTEXT(\"[DC128C414BB6600D6E3E1FB639C86A94]\", \"B50AF5A2408A6A5EBF7BDC9D9E500ECA\", \"Please select or create new save file before create new spline.\")") "Ok" "Ok"))))

(fn DisableCameraOffsetPreview
  (for _array_element (Variables|3.Splines|GetSplineList)
    (Class|SMGSpline|TurnOffCameraOffset _array_element)))

(fn Load
  (bind _savefile (Variables|1.Setup|GetSaveFile))
  (Utilities|IsValid (Variables|1.Setup|GetSaveFile)
    (:Is%20Valid
      (bind _returnvalue (EditorScripting|MessageDialog|ShowMessageDialog (Utilities|Text|MakeLiteralText "NSLOCTEXT(\"[DC128C414BB6600D6E3E1FB639C86A94]\", \"79B570B5459E38433965759D45C26665\", \"Warning\")") (Utilities|Text|MakeLiteralText "NSLOCTEXT(\"[DC128C414BB6600D6E3E1FB639C86A94]\", \"080B64704B33B7A5E435DFB927420EAE\", \"This operation cannot be undo.The unsaved progress will be lost permanently.Confirm loading?\")") "YesNo" "No"))
      (switch Utilities|FlowControl|Switch|SwitchonEAppReturnType _returnvalue
        (:Yes
          (Utilities|IsValid _savefile
            (:Is%20Valid
              (Variables|2.SkeletalMeshSetting%28Optional%29|SetTargetSkeletalMesh%28Optional%29 (Variables|/Game/Tools/SplineMeshGenerator/FX_SplineMeshGenData.FX_SplineMeshGenData_C|GetTargetSkeletalMesh _savefile))
              (Variables|2.SkeletalMeshSetting%28Optional%29|SetTargetBone%28Optional%29 (Variables|/Game/Tools/SplineMeshGenerator/FX_SplineMeshGenData.FX_SplineMeshGenData_C|GetTargetSocket _savefile))
              (Variables|4.NiagaraHelper|SetCameraOffsetPreviewValue (Variables|/Game/Tools/SplineMeshGenerator/FX_SplineMeshGenData.FX_SplineMeshGenData_C|GetCameraOffsetPreviewValue _savefile))
              (Variables|5.Advance|SetbWriteSeededRandomValueIntoVertColorR (Variables|/Game/Tools/SplineMeshGenerator/FX_SplineMeshGenData.FX_SplineMeshGenData_C|GetbWriteSeededRandomValueIntoVertColorR _savefile))
              (Variables|5.Advance|SetbWriteSeededRandomValueIntoVertColorA (Variables|/Game/Tools/SplineMeshGenerator/FX_SplineMeshGenData.FX_SplineMeshGenData_C|GetbWriteSeededRandomValueIntoVertColorA _savefile))
              (Variables|5.Advance|SetRandomSeed (Variables|/Game/Tools/SplineMeshGenerator/FX_SplineMeshGenData.FX_SplineMeshGenData_C|GetRandomSeed _savefile))
              (Variables|5.Advance|SetbLODGeneration (Variables|/Game/Tools/SplineMeshGenerator/FX_SplineMeshGenData.FX_SplineMeshGenData_C|GetbLODGeneration _savefile))
              (Variables|5.Advance|SetLOD1SegReduceFactor(Med) (Variables|/Game/Tools/SplineMeshGenerator/FX_SplineMeshGenData.FX_SplineMeshGenData_C|GetLOD1SegF _savefile))
              (Variables|5.Advance|SetLOD1DivideReduceFactor(Med) (Variables|/Game/Tools/SplineMeshGenerator/FX_SplineMeshGenData.FX_SplineMeshGenData_C|GetLOD1DivF _savefile))
              (Variables|5.Advance|SetLOD2SegReduceFactor(Low) (Variables|/Game/Tools/SplineMeshGenerator/FX_SplineMeshGenData.FX_SplineMeshGenData_C|GetLOD2SegF _savefile))
              (Variables|5.Advance|SetLOD2DivideReduceFactor(Low) (Variables|/Game/Tools/SplineMeshGenerator/FX_SplineMeshGenData.FX_SplineMeshGenData_C|GetLOD2DivF _savefile))
              (CallFunction|RespawnSplineListBaseonSaveFile 0)
              (CallFunction|RefreshSkMesh))
            (:Is%20Not%20Valid
              (Utilities|Array|Clear (Variables|3.Splines|GetSplineList))
              (CallFunction|RefreshSplineListStat)
              (CallFunction|RefreshSkMesh))))))
    (:Is%20Not%20Valid
      (EditorScripting|MessageDialog|ShowMessageDialog (Utilities|Text|MakeLiteralText "NSLOCTEXT(\"[DC128C414BB6600D6E3E1FB639C86A94]\", \"79B570B5459E38433965759D45C26665\", \"Warning\")") (Utilities|Text|MakeLiteralText "NSLOCTEXT(\"[DC128C414BB6600D6E3E1FB639C86A94]\", \"6B1FD1A84D77A34B1E9624AA23A71A89\", \"Please select a save file.\")") "Ok" "Ok"))))

(fn LoadBak_1
  (CallFunction|LoadSplinesfromBackupSlot 0))

(fn LoadBak_2
  (CallFunction|LoadSplinesfromBackupSlot 1))

(fn LoadBak_3
  (CallFunction|LoadSplinesfromBackupSlot 2))

(fn LoadBak_4
  (CallFunction|LoadSplinesfromBackupSlot 3))

(fn LoadSplinesFromBackupSlot
  (params (SlotIndex :type Integer))
  (Utilities|IsValid (Variables|1.Setup|GetSaveFile)
    (:Is%20Valid
      (bind _returnvalue (EditorScripting|MessageDialog|ShowMessageDialog (Utilities|Text|MakeLiteralText "NSLOCTEXT(\"[DC128C414BB6600D6E3E1FB639C86A94]\", \"79B570B5459E38433965759D45C26665\", \"Warning\")") (Utilities|Text|MakeLiteralText "NSLOCTEXT(\"[DC128C414BB6600D6E3E1FB639C86A94]\", \"080B64704B33B7A5E435DFB927420EAE\", \"This operation cannot be undo.The unsaved progress will be lost permanently.Confirm loading?\")") "YesNo" "No"))
      (switch Utilities|FlowControl|Switch|SwitchonEAppReturnType _returnvalue
        (:Yes
          (switch int SlotIndex
            (:0
              (CallFunction|RespawnSplineListBaseonSaveFile 1))
            (:1
              (CallFunction|RespawnSplineListBaseonSaveFile 2))
            (:2
              (CallFunction|RespawnSplineListBaseonSaveFile 3))
            (:3
              (CallFunction|RespawnSplineListBaseonSaveFile 4))))))
    (:Is%20Not%20Valid
      (EditorScripting|MessageDialog|ShowMessageDialog (Utilities|Text|MakeLiteralText "NSLOCTEXT(\"[DC128C414BB6600D6E3E1FB639C86A94]\", \"79B570B5459E38433965759D45C26665\", \"Warning\")") (Utilities|Text|MakeLiteralText "NSLOCTEXT(\"[DC128C414BB6600D6E3E1FB639C86A94]\", \"6B1FD1A84D77A34B1E9624AA23A71A89\", \"Please select a save file.\")") "Ok" "Ok"))))

(fn RefreshSkMesh
  (bind _targetskeletalmesh_optional (Variables|2.SkeletalMeshSetting%28Optional%29|GetTargetSkeletalMesh%28Optional%29))
  (bind _tmptargetskmesh (Variables|Hide|GettmpTargetSkMesh))
  (bind _tmptargetskmesh_2 (Variables|Hide|GettmpTargetSkMesh))
  (Utilities|IsValid _targetskeletalmesh_optional
    (:Is%20Valid
      (Utilities|IsValid _tmptargetskmesh
        (:Is%20Valid
          _)
        (:Is%20Not%20Valid
          (bind _returnvalue (AddComponent|NODE_AddSkeletalMeshComponent-1))
          (Variables|Hide|SettmpTargetSkMesh _returnvalue)))
      (Components|SkeletalMesh|SetSkeletalMeshAsset _tmptargetskmesh_2 _targetskeletalmesh_optional)
      (Components|SkinnedMesh|SetForcedLOD _tmptargetskmesh_2 1)
      (Transformation|SetWorldTransform (Variables|Default|GetTargetDynamicMesh) (Math|Transform|MakeTransform (.location (Transformation|GetSocketTransform _tmptargetskmesh_2 (Variables|2.SkeletalMeshSetting%28Optional%29|GetTargetBone%28Optional%29) "RTS_World")) (.rotation (Transformation|GetSocketTransform _tmptargetskmesh_2 (Variables|2.SkeletalMeshSetting%28Optional%29|GetTargetBone%28Optional%29) "RTS_World"))) false false))
    (:Is%20Not%20Valid
      (Utilities|IsValid _tmptargetskmesh
        (:Is%20Valid
          (Components|DestroyComponent _tmptargetskmesh))
        (:Is%20Not%20Valid
          _))
      (Transformation|SetWorldTransform (Variables|Default|GetTargetDynamicMesh) (Math|Transform|MakeTransform (.location (Transformation|GetActorTransform)) (.rotation (Transformation|GetActorTransform))) false false))))

(fn RefreshSplineListStat
  (bind _addexistspline (Variables|3.Splines|GetAddExistSpline))
  (bind _tmpsplinelist (LocalVariables|GettmpSplineList))
  (Utilities|IsValid _addexistspline
    (:Is%20Valid
      (Utilities|Array|AddUnique (Variables|3.Splines|GetSplineList) _addexistspline)
      (Class|SMGSpline|CheckReftoCurveColorforCopy _addexistspline)
      (Variables|3.Splines|SetAddExistSpline 0))
    (:Is%20Not%20Valid
      ;[1 1 1] Remove repeat and null spline
      (for _array_element (Variables|3.Splines|GetSplineList)
        (Utilities|IsValid _array_element
          (:Is%20Valid
            (if (not (Utilities|Array|ContainsItem _tmpsplinelist _array_element))
              (Utilities|Array|AddUnique _tmpsplinelist _array_element)))))
      (Variables|3.Splines|SetSplineList (LocalVariables|GettmpSplineList))
      ;[1 1 1] Call func on splines before
      (for _array_element_2 (Variables|Hide|GetLastFrameSplineList)
        (Utilities|IsValid _array_element_2
          (:Is%20Valid
            (Class|SMGSpline|ParentRebuild _array_element_2))))
      ;[1 1 1] Assgin self to new splines
      (bind _output_get (Variables|Hide|SetLastFrameSplineList (Variables|3.Splines|GetSplineList)))
      (for _array_element_3 _output_get
        (Utilities|Array|AddUnique (Variables|/Game/Tools/SplineMeshGenerator/SMG_Spline.SMG_Spline_C|GetMyParents _array_element_3) self)
        (Class|SMGSpline|CheckReftoCurveColorforCopy _array_element_3)))))

(fn RespawnSplineListBaseOnSaveFile
  (params (Slot :type Integer))
  (Utilities|Array|Clear (Variables|3.Splines|GetSplineList))
  (for _array_element (Utilities|Select :Index Slot :Option%200 (Variables|/Game/Tools/SplineMeshGenerator/FX_SplineMeshGenData.FX_SplineMeshGenData_C|GetSplines (Variables|1.Setup|GetSaveFile)) :Option%201 (Variables|/Game/Tools/SplineMeshGenerator/FX_SplineMeshGenData.FX_SplineMeshGenData_C|GetSplines_Backup_01 (Variables|1.Setup|GetSaveFile)) :Option%202 (Variables|/Game/Tools/SplineMeshGenerator/FX_SplineMeshGenData.FX_SplineMeshGenData_C|GetSplines_Backup_02 (Variables|1.Setup|GetSaveFile)) :Option%203 (Variables|/Game/Tools/SplineMeshGenerator/FX_SplineMeshGenData.FX_SplineMeshGenData_C|GetSplines_Backup_03 (Variables|1.Setup|GetSaveFile)) :Option%204 (Variables|/Game/Tools/SplineMeshGenerator/FX_SplineMeshGenData.FX_SplineMeshGenData_C|GetSplines_Backup_04 (Variables|1.Setup|GetSaveFile)))
    (bind _returnvalue (Game|SpawnActorfromClass "/Game/Tools/SplineMeshGenerator/SMG_Spline.SMG_Spline_C" (Math|Transform|ComposeTransforms (.RTrans_17_03B6DED849F455BDB1D03EB8048BE264 (Utilities|Struct|BreakSMGSplineSave _array_element)) (Transformation|GetActorTransform)) "AlwaysSpawn" "MultiplyWithRoot" 0 false 0))
    (Utilities|Array|Add (Variables|3.Splines|GetSplineList) _returnvalue)
    (Variables|/Game/Tools/SplineMeshGenerator/SMG_Spline.SMG_Spline_C|SetMaterials (.Materials_14_4251FA1F4009EA32C39497BF1E497ABC (Utilities|Struct|BreakSMGSplineSave _array_element)) _returnvalue)
    (Variables|/Game/Tools/SplineMeshGenerator/SMG_Spline.SMG_Spline_C|SetSegment (.Segments_10_BBEDA2F24E229F5B111E15A3ACF3680C (Utilities|Struct|BreakSMGSplineSave _array_element)) _returnvalue)
    (Variables|/Game/Tools/SplineMeshGenerator/SMG_Spline.SMG_Spline_C|SetDivide (.Divide_9_C478DEF14CBF7FB031B8C08E65F5712A (Utilities|Struct|BreakSMGSplineSave _array_element)) _returnvalue)
    (Variables|/Game/Tools/SplineMeshGenerator/SMG_Spline.SMG_Spline_C|SetVertColor_B (.VertC_B_22_FA2EF92F4E5AA225188BB3987C6BDF36 (Utilities|Struct|BreakSMGSplineSave _array_element)) _returnvalue)
    (Variables|/Game/Tools/SplineMeshGenerator/SMG_Spline.SMG_Spline_C|SetVertColor_R (.VertC_R_29_E5C0090A4E6A9E4A8768A681E46F687C (Utilities|Struct|BreakSMGSplineSave _array_element)) _returnvalue)
    (Variables|/Game/Tools/SplineMeshGenerator/SMG_Spline.SMG_Spline_C|SetVertColor_G (.VertC_G_30_580B4D5D4DA395F7740D4D942B26BB96 (Utilities|Struct|BreakSMGSplineSave _array_element)) _returnvalue)
    (Variables|/Game/Tools/SplineMeshGenerator/SMG_Spline.SMG_Spline_C|SetVertColor_A (.VertC_A_31_65D0505C48AD73F0F5958B9234C06E90 (Utilities|Struct|BreakSMGSplineSave _array_element)) _returnvalue)
    (Variables|/Game/Tools/SplineMeshGenerator/SMG_Spline.SMG_Spline_C|SetSwizzleUV (.SwizzleUV_33_8ABB54E54B98D5203373C182A5DE6E03 (Utilities|Struct|BreakSMGSplineSave _array_element)) _returnvalue)
    (Variables|/Game/Tools/SplineMeshGenerator/SMG_Spline.SMG_Spline_C|SetVertColorCurve (.VertColorCurve_36_32931E3B4F32E64DC679AABFB7092D3B (Utilities|Struct|BreakSMGSplineSave _array_element)) _returnvalue)
    (Spline|ClearSplinePoints (Variables|/Game/Tools/SplineMeshGenerator/SMG_Spline.SMG_Spline_C|GetSpline _returnvalue) true)
    (for (_array_element_2 _array_index) (Utilities|Struct|BreakSMGSplineSave _array_element)
      (Spline|AddSplinePoint (Variables|/Game/Tools/SplineMeshGenerator/SMG_Spline.SMG_Spline_C|GetSpline _returnvalue) (Utilities|Struct|BreakSMGSplinePoint _array_element_2) "Local" false)
      (Spline|SetRotationatSplinePoint (Variables|/Game/Tools/SplineMeshGenerator/SMG_Spline.SMG_Spline_C|GetSpline _returnvalue) _array_index (.Rotation_7_2FCB0B7E47D00C02C8706E8A26CFF749 (Utilities|Struct|BreakSMGSplinePoint _array_element_2)) "Local" false)
      (Spline|SetScaleatSplinePoint (Variables|/Game/Tools/SplineMeshGenerator/SMG_Spline.SMG_Spline_C|GetSpline _returnvalue) _array_index (.Scale_10_8B6343D345A164F46DD1BBA2CA029556 (Utilities|Struct|BreakSMGSplinePoint _array_element_2)) false)
      (Spline|SetTangentsatSplinePoint (Variables|/Game/Tools/SplineMeshGenerator/SMG_Spline.SMG_Spline_C|GetSpline _returnvalue) _array_index (.ArriveTangent_12_026143944714ADE1255152B39EC9CFB4 (Utilities|Struct|BreakSMGSplinePoint _array_element_2)) (.LeaveTangent_14_3BDE9E8B4F4AD0EF17FCD98029311C0A (Utilities|Struct|BreakSMGSplinePoint _array_element_2)) "Local" false)
      (Spline|SetSplinePointType (Variables|/Game/Tools/SplineMeshGenerator/SMG_Spline.SMG_Spline_C|GetSpline _returnvalue) _array_index (.Type_17_598A65944B7309826C0C1CBEAD3FE16C (Utilities|Struct|BreakSMGSplinePoint _array_element_2)) true)))
  (CallFunction|RefreshSplineListStat))

(fn SaveAndBake
  (bind _savefile (Variables|1.Setup|GetSaveFile))
  (bind _targetstaticmesh (Variables|/Game/Tools/SplineMeshGenerator/FX_SplineMeshGenData.FX_SplineMeshGenData_C|GetTargetStaticMesh _savefile))
  (Utilities|IsValid (Variables|1.Setup|GetSaveFile)
    (:Is%20Valid
      (if (== (Utilities|Array|Length (Variables|3.Splines|GetSplineList)) 0)
        (EditorScripting|MessageDialog|ShowMessageDialog (Utilities|Text|MakeLiteralText "NSLOCTEXT(\"[DC128C414BB6600D6E3E1FB639C86A94]\", \"79B570B5459E38433965759D45C26665\", \"Warning\")") (Utilities|Text|MakeLiteralText "NSLOCTEXT(\"[DC128C414BB6600D6E3E1FB639C86A94]\", \"2EE95DA143BA31AA6D98BFB03F4FCF8A\", \"No splines to save and bake.\")") "Ok" "Ok")
        (else
          (bind _returnvalue (EditorScripting|MessageDialog|ShowMessageDialog (Utilities|Text|MakeLiteralText "NSLOCTEXT(\"[DC128C414BB6600D6E3E1FB639C86A94]\", \"79B570B5459E38433965759D45C26665\", \"Warning\")") (Utilities|Text|MakeLiteralText "NSLOCTEXT(\"[DC128C414BB6600D6E3E1FB639C86A94]\", \"4419261E42D6105065ECDDBE7E5FEA69\", \"This operation cannot be undo.The previous save will be overwritten.Confirm saving?\")") "YesNo" "No"))
          (switch Utilities|FlowControl|Switch|SwitchonEAppReturnType _returnvalue
            (:Yes
              (CallFunction|SaveSplinestoSaveFile)
              (Class|SMGEditorHelper|PSetDataAssetDirty (Variables|Default|GetSMG_EditorHelper) (Variables|1.Setup|GetSaveFile))
              (CallFunction|RefreshSkMesh)
              (Class|SMGTickHelperInterface|BaketoStaticMesh)
              ;[1 1 1] For perf
              (if (Variables|1.Setup|Get%28Advance%29AutoSaveAssets)
                (EditorScripting|Asset|SaveLoadedAsset _savefile false :!class /Script/EditorScriptingUtilities.EditorAssetLibrary)
                (Utilities|IsValid _targetstaticmesh
                  (:Is%20Valid
                    (EditorScripting|Asset|SaveLoadedAsset _targetstaticmesh false :!class /Script/EditorScriptingUtilities.EditorAssetLibrary)))))))))
    (:Is%20Not%20Valid
      (EditorScripting|MessageDialog|ShowMessageDialog (Utilities|Text|MakeLiteralText "NSLOCTEXT(\"[DC128C414BB6600D6E3E1FB639C86A94]\", \"79B570B5459E38433965759D45C26665\", \"Warning\")") (Utilities|Text|MakeLiteralText "NSLOCTEXT(\"[DC128C414BB6600D6E3E1FB639C86A94]\", \"101C56924D8FE948BDE0F882FBE0FC96\", \"Please select or create new save file.\")") "Ok" "Ok"))))

(fn SaveSplinesToBackupSlot
  (params (SlotIndex :type Integer))
  (bind _savefile (Variables|1.Setup|GetSaveFile))
  (Utilities|IsValid (Variables|1.Setup|GetSaveFile)
    (:Is%20Valid
      (if (== (Utilities|Array|Length (Variables|3.Splines|GetSplineList)) 0)
        (EditorScripting|MessageDialog|ShowMessageDialog (Utilities|Text|MakeLiteralText "NSLOCTEXT(\"[DC128C414BB6600D6E3E1FB639C86A94]\", \"79B570B5459E38433965759D45C26665\", \"Warning\")") (Utilities|Text|MakeLiteralText "NSLOCTEXT(\"[DC128C414BB6600D6E3E1FB639C86A94]\", \"2EE95DA143BA31AA6D98BFB03F4FCF8A\", \"No splines to save and bake.\")") "Ok" "Ok")
        (else
          (bind _returnvalue (EditorScripting|MessageDialog|ShowMessageDialog (Utilities|Text|MakeLiteralText "NSLOCTEXT(\"[DC128C414BB6600D6E3E1FB639C86A94]\", \"79B570B5459E38433965759D45C26665\", \"Warning\")") (Utilities|Text|MakeLiteralText "NSLOCTEXT(\"[DC128C414BB6600D6E3E1FB639C86A94]\", \"4419261E42D6105065ECDDBE7E5FEA69\", \"This operation cannot be undo.The previous save will be overwritten.Confirm saving?\")") "YesNo" "No"))
          (switch Utilities|FlowControl|Switch|SwitchonEAppReturnType _returnvalue
            (:Yes
              (bind _splines (CallFunction|CollectSplineSaveList))
              (Class|SMGEditorHelper|PSetDataAssetDirty (Variables|Default|GetSMG_EditorHelper) _savefile)
              (switch int (Variables|GetSlotIndex)
                (:0
                  (Variables|/Game/Tools/SplineMeshGenerator/FX_SplineMeshGenData.FX_SplineMeshGenData_C|SetSplines_Backup_01 _splines _savefile))
                (:1
                  (Variables|/Game/Tools/SplineMeshGenerator/FX_SplineMeshGenData.FX_SplineMeshGenData_C|SetSplines_Backup_02 _splines _savefile))
                (:2
                  (Variables|/Game/Tools/SplineMeshGenerator/FX_SplineMeshGenData.FX_SplineMeshGenData_C|SetSplines_Backup_03 _splines _savefile))
                (:3
                  (Variables|/Game/Tools/SplineMeshGenerator/FX_SplineMeshGenData.FX_SplineMeshGenData_C|SetSplines_Backup_04 _splines _savefile))))))))
    (:Is%20Not%20Valid
      (EditorScripting|MessageDialog|ShowMessageDialog (Utilities|Text|MakeLiteralText "NSLOCTEXT(\"[DC128C414BB6600D6E3E1FB639C86A94]\", \"79B570B5459E38433965759D45C26665\", \"Warning\")") (Utilities|Text|MakeLiteralText "NSLOCTEXT(\"[DC128C414BB6600D6E3E1FB639C86A94]\", \"101C56924D8FE948BDE0F882FBE0FC96\", \"Please select or create new save file.\")") "Ok" "Ok"))))

(fn SaveSplinesToSaveFile
  (bind _savefile (Variables|1.Setup|GetSaveFile))
  (bind _savefile_2 (Variables|1.Setup|GetSaveFile))
  (Utilities|IsValid (Variables|1.Setup|GetSaveFile)
    (:Is%20Valid
      (Utilities|Array|Clear (LocalVariables|GettmpSplines))
      (for _array_element (Variables|3.Splines|GetSplineList)
        (Utilities|Array|Clear (LocalVariables|GettmpPoints))
        (for _index (range (Spline|GetNumberOfSplinePoints (Variables|/Game/Tools/SplineMeshGenerator/SMG_Spline.SMG_Spline_C|GetSpline _array_element)))
          (Utilities|Array|Add (LocalVariables|GettmpPoints) (Utilities|Struct|MakeSMGSplinePoint (Spline|GetLocationatSplinePoint (Variables|/Game/Tools/SplineMeshGenerator/SMG_Spline.SMG_Spline_C|GetSpline _array_element) _index "Local") (Spline|GetRotationatSplinePoint (Variables|/Game/Tools/SplineMeshGenerator/SMG_Spline.SMG_Spline_C|GetSpline _array_element) _index "Local") (Spline|GetScaleatSplinePoint (Variables|/Game/Tools/SplineMeshGenerator/SMG_Spline.SMG_Spline_C|GetSpline _array_element) _index) (Spline|GetArriveTangentatSplinePoint (Variables|/Game/Tools/SplineMeshGenerator/SMG_Spline.SMG_Spline_C|GetSpline _array_element) _index "Local") (Spline|GetLeaveTangentatSplinePoint (Variables|/Game/Tools/SplineMeshGenerator/SMG_Spline.SMG_Spline_C|GetSpline _array_element) _index "Local") (Spline|GetSplinePointType (Variables|/Game/Tools/SplineMeshGenerator/SMG_Spline.SMG_Spline_C|GetSpline _array_element) _index))))
        (Utilities|Array|Add (LocalVariables|GettmpSplines) (Utilities|Struct|MakeSMGSplineSave (LocalVariables|GettmpPoints) (Variables|/Game/Tools/SplineMeshGenerator/SMG_Spline.SMG_Spline_C|GetSegment _array_element) (Variables|/Game/Tools/SplineMeshGenerator/SMG_Spline.SMG_Spline_C|GetDivide _array_element) (Variables|/Game/Tools/SplineMeshGenerator/SMG_Spline.SMG_Spline_C|GetMaterials _array_element) (Math|Transform|ComposeTransforms (Transformation|GetActorTransform _array_element) (Math|Transform|InvertTransform (Transformation|GetActorTransform))) (Variables|/Game/Tools/SplineMeshGenerator/SMG_Spline.SMG_Spline_C|GetVertColor_B _array_element) (Variables|/Game/Tools/SplineMeshGenerator/SMG_Spline.SMG_Spline_C|GetVertColor_R _array_element) (Variables|/Game/Tools/SplineMeshGenerator/SMG_Spline.SMG_Spline_C|GetVertColor_G _array_element) (Variables|/Game/Tools/SplineMeshGenerator/SMG_Spline.SMG_Spline_C|GetVertColor_A _array_element) (Variables|/Game/Tools/SplineMeshGenerator/SMG_Spline.SMG_Spline_C|GetSwizzleUV _array_element) (Variables|/Game/Tools/SplineMeshGenerator/SMG_Spline.SMG_Spline_C|GetVertColorCurve _array_element))))
      (Utilities|FlowControl|Sequence
        (:then_0
          (Variables|/Game/Tools/SplineMeshGenerator/FX_SplineMeshGenData.FX_SplineMeshGenData_C|SetSplines (LocalVariables|GettmpSplines) _savefile)
          (Variables|/Game/Tools/SplineMeshGenerator/FX_SplineMeshGenData.FX_SplineMeshGenData_C|SetTargetSkeletalMesh (Variables|2.SkeletalMeshSetting%28Optional%29|GetTargetSkeletalMesh%28Optional%29) _savefile)
          (Variables|/Game/Tools/SplineMeshGenerator/FX_SplineMeshGenData.FX_SplineMeshGenData_C|SetTargetSocket (Variables|2.SkeletalMeshSetting%28Optional%29|GetTargetBone%28Optional%29) _savefile)
          (Variables|/Game/Tools/SplineMeshGenerator/FX_SplineMeshGenData.FX_SplineMeshGenData_C|SetCameraOffsetPreviewValue (Variables|4.NiagaraHelper|GetCameraOffsetPreviewValue) _savefile_2)
          (Variables|/Game/Tools/SplineMeshGenerator/FX_SplineMeshGenData.FX_SplineMeshGenData_C|SetbWriteSeededRandomValueIntoVertColorR (Variables|5.Advance|GetbWriteSeededRandomValueIntoVertColorR) _savefile_2)
          (Variables|/Game/Tools/SplineMeshGenerator/FX_SplineMeshGenData.FX_SplineMeshGenData_C|SetbWriteSeededRandomValueIntoVertColorA (Variables|5.Advance|GetbWriteSeededRandomValueIntoVertColorA) _savefile_2)
          (Variables|/Game/Tools/SplineMeshGenerator/FX_SplineMeshGenData.FX_SplineMeshGenData_C|SetRandomSeed (Variables|5.Advance|GetRandomSeed) _savefile_2)
          (Variables|/Game/Tools/SplineMeshGenerator/FX_SplineMeshGenData.FX_SplineMeshGenData_C|SetbLODGeneration (Variables|5.Advance|GetbLODGeneration) _savefile_2)
          (Variables|/Game/Tools/SplineMeshGenerator/FX_SplineMeshGenData.FX_SplineMeshGenData_C|SetLOD1SegF (Variables|5.Advance|GetLOD1SegReduceFactor(Med)) _savefile_2)
          (Variables|/Game/Tools/SplineMeshGenerator/FX_SplineMeshGenData.FX_SplineMeshGenData_C|SetLOD1DivF (Variables|5.Advance|GetLOD1DivideReduceFactor(Med)) _savefile_2)
          (Variables|/Game/Tools/SplineMeshGenerator/FX_SplineMeshGenData.FX_SplineMeshGenData_C|SetLOD2SegF (Variables|5.Advance|GetLOD2SegReduceFactor(Low)) _savefile_2)
          (Variables|/Game/Tools/SplineMeshGenerator/FX_SplineMeshGenData.FX_SplineMeshGenData_C|SetLOD2DivF (Variables|5.Advance|GetLOD2DivideReduceFactor(Low)) _savefile_2))))))

(fn SetupCameraOffsetPreview
  (for _array_element (Variables|3.Splines|GetSplineList)
    (Class|SMGSpline|TurnonCameraOffset _array_element (Variables|4.NiagaraHelper|GetCameraOffsetPreviewValue))))

(fn ConstructionScript
  (Utilities|FlowControl|Sequence
    (:then_0
      (Class|SMGEditorHelper|PInitialSelectionChangeDetector (Variables|Default|GetSMG_EditorHelper) self))
    (:then_1
      (Variables|/Game/Tools/SplineMeshGenerator/SMG_TickHelper.SMG_TickHelper_C|SetParent self (Variables|Default|GetSMG_TickHelper)))))
