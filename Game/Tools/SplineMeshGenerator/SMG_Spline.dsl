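; SMG_Spline is a spline mesh generator tool actor that sweeps a subdivided strip along a spline with Geometry Script.
; Rebuild makes one mesh per material with UV swizzle and vertex colours, merges them, then schedules an SMG_Parent update.
; Spline-derived normals and tangents replace computed ones, also in reduced-density LOD meshes.
; Editor helpers export static meshes, preview camera offsets, colour splines by parent reference count and destroy orphans outside standalone mode.
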(event Events|EventOnRebuildGeneratedMesh
  (params (TargetMesh :type Dynamic%20Mesh%20Object%20Reference))
  (bind _divide (Variables|Default|GetDivide))
  (bind (_verts _u) (CallFunction|GetPolyLineVertsandU _divide))
  (bind _materials (Variables|Default|GetMaterials))
  (bind _tmptransitmeshes (Variables|Hide|GettmpTransitMeshes))
  (bind _tmpvertcolor (Variables|Hide|GettmpVertColor))
  (bind _dynamicmeshcomponent (Variables|DynamicMeshActor|GetDynamicMeshComponent))
  (bind _visualdynamicmesh (Variables|Default|GetVisualDynamicMesh))
  ;[1 1 1] Comment
  (Utilities|FlowControl|Sequence
    (:then_0
      (Variables|/Game/Tools/SplineMeshGenerator/SMG_TickHelper.SMG_TickHelper_C|SetParent self (Variables|Default|GetSMG_TickHelper)))
    (:then_1
      (Variables|Hide|SettmpVerts _verts)
      (Variables|Hide|SettmpU _u)
      (Variables|Hide|SettmpWeepPoints (CallFunction|GetSweepPoints (Variables|Default|GetSegment))))
    (:then_2
      (DynamicMeshComponent|SetTangentsType (Variables|Default|GetVisualDynamicMesh) "AutoCalculated"))
    (:then_3
      (for _array_element _tmptransitmeshes
        (Components|DestroyComponent _array_element))
      (Utilities|Array|Clear _tmptransitmeshes))
    (:then_4
      (for (_array_element_2 _array_index) (Variables|Default|GetMaterials)
        (Utilities|IsValid _array_element_2
          (:Is%20Valid
            (bind _returnvalue (Game|AddComponentbyClass :Class "/Script/GeometryFramework.DynamicMeshComponent"))
            (Utilities|Array|Add (Variables|Hide|GettmpTransitMeshes) _returnvalue)
            (bind _returnvalue_2 (DynamicMeshComponent|GetDynamicMesh _returnvalue))
            (bind _returnvalue_3 (GeometryScript|Materials|EnableMaterialIDs _returnvalue_2))
            (bind _returnvalue_4 (GeometryScript|Primitives|AppendSweepPolyline :TargetMesh _returnvalue_3 :Transform (Math|Transform|MakeTransform "0, 0, 0") :PolylineVertices (Variables|Hide|GettmpVerts) :SweepPath (Variables|Hide|GettmpWeepPoints) :PolylineTexParamU (Variables|Hide|GettmpU) :bLoop false :StartScale 1.0 :EndScale 1.0 :RotationAngleDeg 0.0 :MiterLimit 1.0))
            (if (Variables|Default|GetSwizzleUV)
              (bind (_selection _returnvalue_5) (GeometryScript|MeshSelection|CreateSelectAllMeshSelection _returnvalue_4 "Vertices"))
              (bind _returnvalue_6 (GeometryScript|UVs|RotateMeshUVs _returnvalue_5 0 9e+01 "(X=0.500000,Y=0.500000)" _selection))
              (GeometryScript|UVs|ScaleMeshUVs _returnvalue_6 0 "(X=-1.000000,Y=1.000000)" "(X=0.500000,Y=0.500000)" _selection))
            (bind _returnvalue_7 (GeometryScript|Normals|RecomputeNormals _returnvalue_4))
            (bind (_triangleidlist _bhastriangleidgaps _returnvalue_8) (GeometryScript|MeshQueries|GetAllTriangleIDs _returnvalue_7))
            (bind _returnvalue_9 (GeometryScript|Materials|SetMaterialIDOnTriangles _returnvalue_8 _triangleidlist _array_index))
            (if (Utilities|IsValid (Variables|Advance|GetVertColorCurve) :!class /Script/Engine.KismetSystemLibrary)
              ;;[1 1 1] Update
              (Utilities|Array|Clear (Variables|Hide|GettmpVertColor))
              (bind (_uvlist _bisvaliduvset _bhasvertexidgaps _bhassplituvs _returnvalue_10) (GeometryScript|UVs|GetMeshPerVertexUVs _returnvalue_9))
              (bind _uvarray (GeometryScript|ListUtils|ConvertUVListtoArray _uvlist))
              (for _array_element_3 _uvarray
                (Utilities|Array|Add _tmpvertcolor (Math|Curves|GetClampedLinearColorValue (Variables|Advance|GetVertColorCurve) (Math|Float|SelectFloat (Math|Vector2D|BreakVector2D _array_element_3) (.Y (Math|Vector2D|BreakVector2D _array_element_3)) (Variables|Default|GetSwizzleUV)))))
              (bind _colorlist (GeometryScript|ListUtils|ConvertArraytoColorList _tmpvertcolor))
              (GeometryScript|VertexColor|SetMeshPerVertexColors _returnvalue_10 _colorlist)
              (else
                (GeometryScript|VertexColor|SetMeshConstantVertexColor _returnvalue_9 (Math|Color|MakeColor (Variables|Advance|GetVertColor_R) (Variables|Advance|GetVertColor_G) (Variables|Advance|GetVertColor_B) (Variables|Advance|GetVertColor_A)) (Utilities|Struct|MakeGeometryScriptColorFlags true) true)))))))
    (:then_5
      (bind _returnvalue_11 (DynamicMeshComponent|GetDynamicMesh _visualdynamicmesh))
      (bind _returnvalue_12 (DynamicMesh|Reset _returnvalue_11))
      (for _array_element_4 (Variables|Hide|GettmpTransitMeshes)
        (bind _returnvalue_13 (DynamicMeshComponent|GetDynamicMesh _array_element_4))
        (GeometryScript|MeshEdits|AppendMesh :TargetMesh _returnvalue_12 :AppendMesh _returnvalue_13 :bDeferChangeNotifications false :AppendOptions "(CombineMode=EnableAllMatching)"))
      (bind _returnvalue_14 (DynamicMeshComponent|GetDynamicMesh _dynamicmeshcomponent))
      (bind _returnvalue_15 (DynamicMesh|Reset _returnvalue_14))
      (bind _returnvalue_16 (DynamicMeshComponent|GetDynamicMesh (Variables|Default|GetVisualDynamicMesh)))
      (GeometryScript|MeshEdits|AppendMesh :TargetMesh _returnvalue_15 :AppendMesh _returnvalue_16 :bDeferChangeNotifications false :AppendOptions "(CombineMode=EnableAllMatching)"))
    (:then_6
      (GeometryScript|Scene|SetComponentMaterialList _visualdynamicmesh _materials)
      (GeometryScript|Scene|SetComponentMaterialList _dynamicmeshcomponent _materials))
    (:then_7
      (Utilities|Time|SetTimerbyFunctionName (Variables|Default|GetSMG_TickHelper) "t_UpdateParent" 0.0333 false false 0.033))))

(event Custom|UpdateCameraOffset
  (params (CameraLocation :type Vector))
  (Transformation|SetWorldLocation (Variables|Default|GetVisualDynamicMesh) (+ (Transformation|GetWorldLocation (Variables|DynamicMeshActor|GetDynamicMeshComponent)) (* (Math|Vector|Normalize (- CameraLocation (Transformation|GetWorldLocation (Variables|DynamicMeshActor|GetDynamicMeshComponent)))) (Variables|Hide|GetCameraOffset))) false false))

(event EventUpdateParent
  ;[1 1 1] Comment
  (for _array_element (Variables|Hide|GetMyParents)
    (Class|SMGParent|SplineUpdate _array_element)))

(event Custom|CorrectTangentAndNormal
  (bind _divide (Variables|Default|GetDivide))
  (bind _segment (Variables|Default|GetSegment))
  (bind (_normallist _tangentxlist _tangentylist) (CallFunction|GetNormalArrayandTangentArray _segment _divide))
  (bind _visualdynamicmesh (Variables|Default|GetVisualDynamicMesh))
  ;[1 1 1] Comment
  (Utilities|FlowControl|Sequence
    (:then_0
      (Variables|Hide|SettmpNormal%20List _normallist)
      (Variables|Hide|SettmpTangent%20XList _tangentxlist)
      (Variables|Hide|SettmpTangent%20YList _tangentylist))
    (:then_1
      (for _array_element (Variables|Hide|GettmpTransitMeshes)
        (bind _returnvalue (DynamicMeshComponent|GetDynamicMesh _array_element))
        (GeometryScript|Normals|SetMeshToPerVertexNormals(FromList) _returnvalue (Variables|Hide|GettmpNormal%20List))
        (GeometryScript|Normals|SetMeshPerVertexTangents _returnvalue (Variables|Hide|GettmpTangent%20XList) (Variables|Hide|GettmpTangent%20YList))))
    (:then_2
      (bind _returnvalue_2 (DynamicMeshComponent|GetDynamicMesh _visualdynamicmesh))
      (bind _returnvalue_3 (DynamicMesh|Reset _returnvalue_2))
      (for _array_element_2 (Variables|Hide|GettmpTransitMeshes)
        (bind _returnvalue_4 (DynamicMeshComponent|GetDynamicMesh _array_element_2))
        (GeometryScript|MeshEdits|AppendMesh :TargetMesh _returnvalue_3 :AppendMesh _returnvalue_4 :bDeferChangeNotifications false :AppendOptions "(CombineMode=EnableAllMatching)"))
      (bind _returnvalue_5 (DynamicMeshComponent|GetDynamicMesh (Variables|DynamicMeshActor|GetDynamicMeshComponent)))
      (bind _returnvalue_6 (DynamicMesh|Reset _returnvalue_5))
      (bind _returnvalue_7 (DynamicMeshComponent|GetDynamicMesh (Variables|Default|GetVisualDynamicMesh)))
      (GeometryScript|MeshEdits|AppendMesh :TargetMesh _returnvalue_6 :AppendMesh _returnvalue_7 :bDeferChangeNotifications false :AppendOptions "(CombineMode=EnableAllMatching)")
      (DynamicMeshComponent|SetTangentsType _visualdynamicmesh "ExternallyProvided"))))

(event Custom|ParentRebuild
  ;[1 1 1] Comment
  (CallFunction|CheckRef)
  (CallFunction|CheckParentValid))

(event EventBakeToStaticMesh)

(fn CheckParentValid
  (for _array_element (Variables|Hide|GetMyParents)
    (Utilities|IsValid _array_element
      (:Is%20Valid
        (LocalVariables|SetbIsValid true))))
  (if (LocalVariables|GetbIsValid)
    _
    (elif (Variables|StandaloneMode|GetDONOTCHANGE_StandAloneMode)
      _
      (else
        (Actor|DestroyActor self)))))

(fn CheckRef
  (for _array_element (Variables|Hide|GetMyParents)
    (if (Utilities|Array|ContainsItem (Variables|/Game/Tools/SplineMeshGenerator/SMG_Parent.SMG_Parent_C|GetSplineList _array_element) self)
      (Utilities|Array|Add (LocalVariables|GettmpParentsList) _array_element)))
  (Variables|Hide|SetMyParents (LocalVariables|GettmpParentsList)))

(fn CheckRefToCurveColorForCopy
  (bind _refcount (LocalVariables|GetRefCount))
  (bind _spline (Variables|Default|GetSpline))
  (for _array_element (Variables|Hide|GetMyParents)
    (if (Utilities|Array|ContainsItem (Variables|/Game/Tools/SplineMeshGenerator/SMG_Parent.SMG_Parent_C|GetSplineList _array_element) self)
      (Math|Integer|IncrementInt (LocalVariables|GetRefCount))))
  (if (== _refcount 1)
    (Editor|SetUnselectedSplineSegmentColor _spline (Math|Color|LinearColorGreen))
    (Spline|UpdateSpline _spline)
    (elif (== _refcount 0)
      (Editor|SetUnselectedSplineSegmentColor _spline (Math|Color|LinearColorRed))
      (Spline|UpdateSpline _spline)
      (else
        (Editor|SetUnselectedSplineSegmentColor _spline (Math|Color|LinearColorYellow))
        (Spline|UpdateSpline _spline)))))

(fn ExportToStaticMesh
  (bind _exportname (Variables|StandaloneMode|GetExportName))
  (bind _exportpath (Variables|StandaloneMode|GetExportPath))
  (bind _returnvalue (Utilities|String|Append _exportpath "/" _exportname))
  (bind _targetsm (LocalVariables|GetTargetSM))
  (Utilities|FlowControl|Sequence
    (:then_0
      (bind _returnvalue_2 (EditorScripting|Asset|DoesAssetExist _returnvalue :!class /Script/EditorScriptingUtilities.EditorAssetLibrary))
      (LocalVariables|SetTargetAssetExist _returnvalue_2))
    (:then_1
      (if (LocalVariables|GetTargetAssetExist)
        (bind _returnvalue_3 (EditorScripting|Asset|LoadAsset _returnvalue :!class /Script/EditorScriptingUtilities.EditorAssetLibrary))
        (bind _asstatic_mesh (Utilities|Casting|CastToStaticMesh _returnvalue_3))
        (LocalVariables|SetTargetSM _asstatic_mesh)
        (else
          (bind _returnvalue_4 (EditorScripting|Asset|DuplicateAsset "/Game/NeverCook/FX/SMG/SMG_TemplateStaticMesh" _returnvalue :!class /Script/EditorScriptingUtilities.EditorAssetLibrary))
          (Utilities|IsValid _returnvalue_4
            (:Is%20Valid
              (bind _asstatic_mesh_2 (Utilities|Casting|CastToStaticMesh _returnvalue_4))
              (LocalVariables|SetTargetSM _asstatic_mesh_2))))))
    (:then_2
      (Utilities|IsValid _targetsm
        (:Is%20Valid
          (bind _returnvalue_5 (DynamicMeshComponent|GetDynamicMesh (Variables|Default|GetVisualDynamicMesh)))
          (GeometryScript|StaticMesh|CopyMeshtoStaticMesh _returnvalue_5 _targetsm (Utilities|Struct|MakeGeometryScriptCopyMeshtoAssetOptions false false false "DoNothing" false false true true "ReplaceAssetMaterials" (Variables|Default|GetMaterials)) (Utilities|Struct|MakeGeometryScriptMeshWriteLOD false)
            (:Success
              (Class|SMGEditorHelper|RemoveStaticMeshLODs (Variables|Default|GetSMG_EditorHelper) (LocalVariables|GetTargetSM)))))))))

(fn GetLODMesh
  (params (SegFactor :type Float%20%28double-precision%29 :default 1.0) (DivideFactor :type Float%20%28double-precision%29 :default 1.0))
  (bind _divide (Variables|Default|GetDivide))
  (bind _returnvalue (Math|Conversions|ToFloat(Integer) _divide))
  (bind _returnvalue_2 (* DivideFactor _returnvalue))
  (bind _returnvalue_3 (Math|Float|Floor _returnvalue_2))
  (bind _segment (Variables|Default|GetSegment))
  (bind _returnvalue_4 (Math|Conversions|ToFloat(Integer) _segment))
  (bind _returnvalue_5 (* SegFactor _returnvalue_4))
  (bind _returnvalue_6 (Math|Float|Floor _returnvalue_5))
  (bind (_verts _u) (CallFunction|GetPolyLineVertsandU _returnvalue_3))
  (bind (_normallist _tangentxlist _tangentylist) (CallFunction|GetNormalArrayandTangentArray _returnvalue_6 _returnvalue_3))
  (bind _ltmpvertcolor (LocalVariables|GetLTmpVertColor))
  (Utilities|FlowControl|Sequence
    (:then_0
      (LocalVariables|SetLTmpVert _verts)
      (LocalVariables|SetLTmpU _u)
      (LocalVariables|SetLTmpSweepPoints (CallFunction|GetSweepPoints _returnvalue_6))
      (LocalVariables|SetLTmpNormalList _normallist)
      (LocalVariables|SetLTmpTangentXList _tangentxlist)
      (LocalVariables|SetLTmpTangentYList _tangentylist))
    (:then_1
      (bind _returnvalue_7 (Game|ConstructObjectfromClass "/Script/GeometryFramework.DynamicMesh"))
      (GeometryScript|Materials|EnableMaterialIDs _returnvalue_7)
      (LocalVariables|SetLReturnMesh _returnvalue_7))
    (:then_2
      (for (_array_element _array_index) (Variables|Default|GetMaterials)
        (Utilities|IsValid _array_element
          (:Is%20Valid
            (bind _returnvalue_8 (Game|ConstructObjectfromClass "/Script/GeometryFramework.DynamicMesh"))
            (Utilities|Array|Add (LocalVariables|GetLTmpMeshList) _returnvalue_8)
            (bind _returnvalue_9 (GeometryScript|Materials|EnableMaterialIDs _returnvalue_8))
            (bind _returnvalue_10 (GeometryScript|Primitives|AppendSweepPolyline :TargetMesh _returnvalue_9 :Transform (Math|Transform|MakeTransform "0, 0, 0") :PolylineVertices (LocalVariables|GetLTmpVert) :SweepPath (LocalVariables|GetLTmpSweepPoints) :PolylineTexParamU (LocalVariables|GetLTmpU) :bLoop false :StartScale 1.0 :EndScale 1.0 :RotationAngleDeg 0.0 :MiterLimit 1.0))
            (if (Variables|Default|GetSwizzleUV)
              (bind (_selection _returnvalue_11) (GeometryScript|MeshSelection|CreateSelectAllMeshSelection _returnvalue_10 "Vertices"))
              (bind _returnvalue_12 (GeometryScript|UVs|RotateMeshUVs _returnvalue_11 0 9e+01 "(X=0.500000,Y=0.500000)" _selection))
              (GeometryScript|UVs|ScaleMeshUVs _returnvalue_12 0 "(X=-1.000000,Y=1.000000)" "(X=0.500000,Y=0.500000)" _selection))
            (bind _returnvalue_13 (GeometryScript|Normals|RecomputeNormals _returnvalue_10))
            (bind (_triangleidlist _bhastriangleidgaps _returnvalue_14) (GeometryScript|MeshQueries|GetAllTriangleIDs _returnvalue_13))
            (bind _returnvalue_15 (GeometryScript|Materials|SetMaterialIDOnTriangles _returnvalue_14 _triangleidlist _array_index))
            (if (Utilities|IsValid (Variables|Advance|GetVertColorCurve) :!class /Script/Engine.KismetSystemLibrary)
              ;[1 1 1] Update
              (Utilities|Array|Clear (LocalVariables|GetLTmpVertColor))
              (bind (_uvlist _bisvaliduvset _bhasvertexidgaps _bhassplituvs _returnvalue_16) (GeometryScript|UVs|GetMeshPerVertexUVs _returnvalue_15))
              (bind _uvarray (GeometryScript|ListUtils|ConvertUVListtoArray _uvlist))
              (for _array_element_2 _uvarray
                (Utilities|Array|Add _ltmpvertcolor (Math|Curves|GetClampedLinearColorValue (Variables|Advance|GetVertColorCurve) (Math|Float|SelectFloat (Math|Vector2D|BreakVector2D _array_element_2) (.Y (Math|Vector2D|BreakVector2D _array_element_2)) (Variables|Default|GetSwizzleUV)))))
              (bind _colorlist (GeometryScript|ListUtils|ConvertArraytoColorList _ltmpvertcolor))
              (bind _returnvalue_17 (GeometryScript|VertexColor|SetMeshPerVertexColors _returnvalue_16 _colorlist))
              (bind _returnvalue_18 (GeometryScript|Normals|SetMeshToPerVertexNormals(FromList) _returnvalue_17 (LocalVariables|GetLTmpNormalList)))
              (GeometryScript|Normals|SetMeshPerVertexTangents _returnvalue_18 (LocalVariables|GetLTmpTangentXList) (LocalVariables|GetLTmpTangentYList))
              (else
                (bind _returnvalue_19 (GeometryScript|VertexColor|SetMeshConstantVertexColor _returnvalue_15 (Math|Color|MakeColor (Variables|Advance|GetVertColor_R) (Variables|Advance|GetVertColor_G) (Variables|Advance|GetVertColor_B) (Variables|Advance|GetVertColor_A)) (Utilities|Struct|MakeGeometryScriptColorFlags true) true))
                (bind _returnvalue_20 (GeometryScript|Normals|SetMeshToPerVertexNormals(FromList) _returnvalue_19 (LocalVariables|GetLTmpNormalList)))
                (GeometryScript|Normals|SetMeshPerVertexTangents _returnvalue_20 (LocalVariables|GetLTmpTangentXList) (LocalVariables|GetLTmpTangentYList))))))))
    (:then_3
      (for _array_element_3 (LocalVariables|GetLTmpMeshList)
        (GeometryScript|MeshEdits|AppendMesh :TargetMesh (LocalVariables|GetLReturnMesh) :AppendMesh _array_element_3 :bDeferChangeNotifications false :AppendOptions "(CombineMode=EnableAllMatching)")))
    (:then_4
      (return (LocalVariables|GetLReturnMesh)))))

(fn GetNormalArrayAndTangentArray
  (params (SegmentInput :type Integer) (DivideInput :type Integer))
  (bind _returnvalue (Math|Conversions|ToFloat(Integer) SegmentInput))
  (bind _swizzleuv (Variables|Default|GetSwizzleUV))
  (bind _spline (Variables|Default|GetSpline))
  (bind _spline_2 (Variables|Default|GetSpline))
  (Utilities|FlowControl|Sequence
    (:then_0
      (for _index (range (+ SegmentInput 1))
        (for _ (range (+ (+ DivideInput 1) 1))
          (Utilities|Array|Add (LocalVariables|GettmpNormalList) (Spline|GetUpVectoratTime _spline (/ (Math|Conversions|ToFloat(Integer) _index) _returnvalue) "Local" true))
          (Utilities|Array|Add (LocalVariables|GettmpTangentXList) (Math|Vector|SelectVector (Math|Vector|CrossProduct (* (Spline|GetRightVectoratTime _spline_2 (/ (Math|Conversions|ToFloat(Integer) _index) _returnvalue) "Local" true) "-1.000000,-1.000000,-1.000000") (Spline|GetUpVectoratTime _spline (/ (Math|Conversions|ToFloat(Integer) _index) _returnvalue) "Local" true)) (* (Spline|GetRightVectoratTime _spline_2 (/ (Math|Conversions|ToFloat(Integer) _index) _returnvalue) "Local" true) "-1.000000,-1.000000,-1.000000") _swizzleuv))
          (Utilities|Array|Add (LocalVariables|GettmpTangentYList) (Math|Vector|SelectVector (* (Spline|GetRightVectoratTime _spline_2 (/ (Math|Conversions|ToFloat(Integer) _index) _returnvalue) "Local" true) "-1.000000,-1.000000,-1.000000") (Math|Vector|CrossProduct (* (Spline|GetRightVectoratTime _spline_2 (/ (Math|Conversions|ToFloat(Integer) _index) _returnvalue) "Local" true) "-1.000000,-1.000000,-1.000000") (Spline|GetUpVectoratTime _spline (/ (Math|Conversions|ToFloat(Integer) _index) _returnvalue) "Local" true)) _swizzleuv)))))
    (:then_1
      (bind _vectorlist (GeometryScript|ListUtils|ConvertArraytoVectorList (LocalVariables|GettmpTangentXList)))
      (bind _vectorlist_2 (GeometryScript|ListUtils|ConvertArraytoVectorList (LocalVariables|GettmpTangentYList)))
      (bind _vectorlist_3 (GeometryScript|ListUtils|ConvertArraytoVectorList (LocalVariables|GettmpNormalList)))
      (return _vectorlist_3 _vectorlist _vectorlist_2))))

(fn GetPolyLineVertsAndU
  (params (DivideInput :type Integer))
  (bind _returnvalue (+ DivideInput 1))
  (bind _returnvalue_2 (Math|Conversions|ToFloat(Integer) _returnvalue))
  (bind _debugscale (Variables|Hide|GetDebugScale))
  (Utilities|Array|Add (LocalVariables|GetPolylineU) 1.0)
  (Utilities|Array|Add (LocalVariables|GetPolylineVert) (Math|Vector2D|MakeVector2D (* 1.0 _debugscale)))
  (for _index (range 1 (+ _returnvalue 1))
    (Utilities|Array|Add (LocalVariables|GetPolylineVert) (Math|Vector2D|MakeVector2D (* (+ (/ (/ (Math|Conversions|ToFloat(Integer) _index) _returnvalue_2) 0.5) -1.0) (* -1.0 _debugscale))))
    (Utilities|Array|Add (LocalVariables|GetPolylineU) (- 1.0 (/ (Math|Conversions|ToFloat(Integer) _index) _returnvalue_2))))
  (Utilities|Array|Reverse (LocalVariables|GetPolylineU))
  (return (LocalVariables|GetPolylineVert) (LocalVariables|GetPolylineU)))

(fn GetSweepPoints
  (params (SegmentInput :type Integer))
  (for _index (range (+ SegmentInput 1))
    (Utilities|Array|Add (LocalVariables|GetSweepP) (Spline|GetTransformatTime (Variables|Default|GetSpline) (/ (Math|Conversions|ToFloat(Integer) _index) (Math|Conversions|ToFloat(Integer) SegmentInput)) "Local" true false)))
  (return (LocalVariables|GetSweepP)))

(fn ParentSaveAndBake
  (for _array_element (Variables|Hide|GetMyParents)
    (1.Setup|SaveandBake _array_element)))

(fn TurnOffCameraOffset
  (Class|SMGEditorHelper|STurnOffCameraOffsetEvent (Variables|Default|GetSMG_EditorHelper))
  (Variables|Hide|SetCameraOffsetPreview false)
  (Transformation|SetWorldLocation (Variables|Default|GetVisualDynamicMesh) (Transformation|GetWorldLocation (Variables|DynamicMeshActor|GetDynamicMeshComponent)) false false))

(fn TurnOnCameraOffset
  (params (CO :type Float%20%28double-precision%29))
  (Variables|Hide|SetCameraOffset CO)
  (Variables|Hide|SetCameraOffsetPreview true)
  (Class|SMGEditorHelper|STurnonCameraOffsetEvent (Variables|Default|GetSMG_EditorHelper) self)
  (bind _cameralocation (Rendering|GetEditorCameraInfo))
  (CallFunction|UpdateCameraOffset _cameralocation))

(fn ConstructionScript
  (Utilities|FlowControl|Sequence
    (:then_0
      (Class|SMGEditorHelper|SInitialSelectionChangeDetector (Variables|Default|GetSMG_EditorHelper) self)
      (if (Variables|Hide|GetCameraOffsetPreview)
        (CallFunction|TurnonCameraOffset (Variables|Hide|GetCameraOffset))))
    (:then_1
      (CallFunction|CheckReftoCurveColorforCopy))))
