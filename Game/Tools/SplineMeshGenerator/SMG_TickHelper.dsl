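; SMG_TickHelper is a helper for the Spline Mesh Generator tool that relays requests to a parent object.
; Its UpdateParent and BakeToStaticMesh functions each get the stored parent, check that it is valid, then send the matching SMGTickHelperInterface message.
; Interface messages carry both calls, so the helper needs no cast to the generator class.
; BeginPlay and Tick are empty, and nothing in this graph calls either function.
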(event EventBeginPlay)

(event EventTick
  (params (DeltaSeconds :type Float%20%28single-precision%29)))

(fn t_BakeToStaticMesh
  (bind _parent (Variables|Default|GetParent))
  (Utilities|IsValid _parent
    (:Is%20Valid
      (Class|SMGTickHelperInterface|BaketoStaticMesh(Message) _parent))))

(fn t_UpdateParent
  (bind _parent (Variables|Default|GetParent))
  (Utilities|IsValid _parent
    (:Is%20Valid
      (Class|SMGTickHelperInterface|UpdateParent(Message) _parent))))
