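; SMG_TickHelperInterface is a Blueprint interface for the SplineMeshGenerator tool.
; It declares two functions, BakeToStaticMesh and UpdateParent, both with no logic in the file.
; The names suggest a tick helper uses it to call back into its parent generator for updates and static mesh baking.
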(fn BakeToStaticMesh)

(fn UpdateParent)
